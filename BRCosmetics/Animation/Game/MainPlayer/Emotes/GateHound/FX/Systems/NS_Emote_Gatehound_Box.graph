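# Niagara effect for the box in the GateHound emote, staging a timed sequence of glows, light beams, smoke and sparks over roughly six seconds.
# Emitters follow sampled skeletal mesh bones or sockets, with User.GlowColor tinting several flares and shine sprites.
# Delayed loops and timed bursts sequence flares with lights, ribbon beams, mesh light shafts and GPU vortex sparks.
# The main flare adjusts size and camera offset by view direction.
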NiagaraSystem /BRCosmetics/Animation/Game/MainPlayer/Emotes/GateHound/FX/Systems/NS_Emote_Gatehound_Box
dependencies {
 usedDataInterfaces [
  &NiagaraDataInterfaceCurve
  &NiagaraDataInterfaceCamera
  &NiagaraDataInterfaceColorCurve
  &NiagaraDataInterfaceVector2DCurve
  &NiagaraDataInterfaceSpriteRendererInfo
  &NiagaraDataInterfaceArrayFloat3
  &NiagaraDataInterfaceArrayInt32
  &NiagaraDataInterfaceArrayFloat
  &NiagaraDataInterfaceSkeletalMesh
 ]
 usedDynamicInputs [&Vector2DFromCurve, &RandomRangeVector2D]
 usedModules [
  &self:SystemSpawnScript.SystemScriptSource.SystemScriptGraph.NiagaraNodeAssignment_0.SetVariables_#1
  &SystemState
 ]
}
emitters {
 BoxFlare {
  emitterUpdateScript [
   &EmitterState(Life Cycle Mode=::NewEnumerator1, Loop Behavior=::NewEnumerator1, Loop Duration=5.9868, UseLoopDelay=true)
   &SpawnBurst_Instantaneous(Spawn Count=2, Spawn Time=5, Use Spawn Probability=false, UseLoopCountLimit=false) {module=SpawnBurst_Instantaneous001}
   &SpawnBurst_Instantaneous(Spawn Count=2, Use Spawn Probability=false, UseLoopCountLimit=false)
  ]
  particleSpawnScript [
   &PartitionParticles(Particles Per Partition=1, Partition Mode:ENiagara_PartitionMode=::NewEnumerator1)
   &InitializeParticle(
    Color=&MultiplyLinearColor(Linear Color=<rgba 1 1 1 1>, Scale Factor=&MakeVector4_FromFloat(Float=1))
    Sprite Size=<16 30>
    Sprite Size Mode=::NewEnumerator4
    UsePositionOffset=true
    Write Material Random=true
   )
   /SetVariables(
    Particles.SpriteAlignment=&SelectVectorFromArray(
     Direct Array Index=$Particles.PartitionParticles.Partition
     Vector Selection Array=DataInterface() {propertyValues {FloatData [<0 0 1>, <0 0 1>]}}
    )
    Particles.SpriteFacing=&SelectVectorFromArray(
     Direct Array Index=$Particles.PartitionParticles.Partition
     Vector Selection Array=DataInterface() {propertyValues {FloatData [<1 1 0>, <-1 1 0>]}}
    )
   ) {isSetParameters=true}
  ]
  particleUpdateScript [
   &SkeletalMeshLocation(
    Bone / Socket Index=1
    Bone Sampling Mode=::NewEnumerator4
    MaskSpawnGroup=false
    Sampled Position Offset Coordinate Space:ENiagaraSkelMeshTransforms=::NewEnumerator0
    Skeletal Mesh=$System.Skeletal Mesh
   )
   &ParticleState()
   &ScaleSpriteSize(
    Non-Uniform Curve Sprite Scale=DataInterface() {propertyValues {
      XCurve {keys [{tangentMode=RCTM_User, time=1, value=1}]}
      YCurve {keys [
        {arriveTangent=4.1225, interpMode=RCIM_Cubic, leaveTangent=4.1225, tangentMode=RCTM_User, time=0, value=0}
        {arriveTangent=0.4477, interpMode=RCIM_Cubic, leaveTangent=0.4477, tangentMode=RCTM_User, time=0.25, value=1}
        {arriveTangent=-1.3333, interpMode=RCIM_Cubic, leaveTangent=-1.3333, time=0.7, value=0.7}
        {tangentMode=RCTM_User, time=1, value=0}
       ]}
     }}
    Scale Sprite Size Mode=::NewEnumerator3
   )
   &ScaleColor(
    Scale Alpha=&FloatFromCurve(FloatCurve=DataInterface() {propertyValues {Curve {keys [
         {arriveTangent=9.0725, interpMode=RCIM_Cubic, leaveTangent=9.0725, tangentMode=RCTM_User, time=0, value=0}
         {arriveTangent=-2.5249, interpMode=RCIM_Cubic, leaveTangent=-2.5249, tangentMode=RCTM_User, time=0.1408, value=0.9847}
         {arriveTangent=-0.0405, interpMode=RCIM_Cubic, leaveTangent=-0.0405, tangentMode=RCTM_User, time=1, value=0}
        ]}}})
    ScaleA=true
    ScaleRGB=true
   )
  ]
  rendererClasses [&NiagaraSpriteRendererProperties]
  renderers [{rendererClass=&NiagaraSpriteRendererProperties}]
 }
 Flare {
  emitterUpdateScript [
   &EmitterState(Inactive Response=::NewEnumerator1, Life Cycle Mode=::NewEnumerator1, Loop Behavior=::NewEnumerator1, Loop Delay=0.75, UseLoopDelay=true)
   &SpawnBurst_Instantaneous(Spawn Count=1, Use Spawn Probability=false)
  ]
  particleSpawnScript [
   &InitializeParticle(
    Color=$User.GlowColor
    Lifetime=4.5
    Sprite Rotation Mode=::NewEnumerator2
    Sprite Size Mode=::NewEnumerator3
    Uniform Sprite Size=100
    UsePositionOffset=true
    Write Material Random=true
   )
   &DynamicMaterialParameters(
    Index 0 Param 1=0.5
    Index 0 Param 2=1
    Index 0 Param 3=1
    Index 0 Param 4=1
    Module.Param0WriteEnabled=true
    Module.Param1WriteEnabled=true
    Module.Param2WriteEnabled=true
    Module.Param3WriteEnabled=true
   )
  ]
  particleUpdateScript [
   &SkeletalMeshLocation(Bone Sampling Mode=::NewEnumerator4, MaskSpawnGroup=false, Skeletal Mesh=$System.Skeletal Mesh)
   &CameraOffset(Camera Offset Amount=&Lerp_Float(
     A=150
     B=25
     Alpha=&Power(
      a=&NormalizeFloat(
       Value=&MakeFloatFromVector(
        Vector=&ReturnCameraVector(Camera Vectors=::NewEnumerator2, Coordinate Space=::Simulation)
        Channel=::NewEnumerator2
       )
       Min=0
       Max=1
      )
      b=0.75
     )
    ))
   &ParticleState()
   &ScaleSpriteSize(
    Uniform Curve Scale=&Lerp_Float(
     A=1.5
     B=1
     Alpha=&Power(
      a=&NormalizeFloat(
       Value=&MakeFloatFromVector(
        Vector=&ReturnCameraVector(Camera Vectors=::NewEnumerator2, Coordinate Space=::Simulation)
        Channel=::NewEnumerator2
       )
       Min=0
       Max=1
      )
      b=1.5
     )
    )
    Uniform Curve Sprite Scale=DataInterface() {propertyValues {Curve {keys [
        {arriveTangent=5.6475, interpMode=RCIM_Cubic, leaveTangent=5.6475, tangentMode=RCTM_User, time=0, value=0}
        {arriveTangent=1.25, time=0.2, value=1}
        {interpMode=RCIM_Cubic, tangentMode=RCTM_User, time=0.8, value=1}
       ]}}}
   )
   &SolveForcesAndVelocity(Clamp Velocity=false, Limit Acceleration=false)
   &ScaleColor(
    Scale Alpha=&Multiply_Float(
     A=1
     B=&FloatFromCurve(FloatCurve=DataInterface() {propertyValues {Curve {keys [<at 0 0>,<at 0.2 1>,<at 0.9 1>,<at 1 0>]}}})
    )
    Scale RGB=&VectorFromFloat(Value=3)
    ScaleA=true
    ScaleRGB=true
   )
   &FN_Light_Attributes(
    Color=<rgba 1 0.3293 0 1>
    Create Color Attribute=true
    Create Exponent Attribute=true
    Create Position Attribute=true
    Create Radius Attribute=true
    Exponent=4
    Position=$Particles.SkeletalMeshLocation.SampledPosition
    Radius=175
    Scale Color=&Waveform(
     Global Amplitude Scale=&FloatFromCurve(
      FloatCurve=DataInterface() {propertyValues {Curve {keys [
          {interpMode=RCIM_Cubic, tangentMode=RCTM_User, time=0, value=0}
          {arriveTangent=1.25, interpMode=RCIM_Cubic, leaveTangent=1.25, time=0.2, value=1}
          {interpMode=RCIM_Cubic, tangentMode=RCTM_User, time=0.8, value=1}
          {tangentMode=RCTM_User, time=1, value=0}
         ]}}}
      Scale Curve=5
     )
     "Waveform [1]"=::NewEnumerator12
     "[1] Frequency"=12
     "[1] Amplitude Min/Max"=<0 1>
     "[1] Bias"=0.5
    )
   )
  ]
  rendererClasses [&NiagaraSpriteRendererProperties, &NiagaraLightRendererProperties]
  renderers [
   {rendererClass=&NiagaraSpriteRendererProperties}
   {rendererClass=&NiagaraLightRendererProperties, rendererIndex=1}
  ]
 }
 FlareSkull {
  emitterUpdateScript [
   &EmitterState(Inactive Response=::NewEnumerator1, Life Cycle Mode=::NewEnumerator1, Loop Behavior=::NewEnumerator1, Loop Duration=5.5, UseLoopDelay=false)
   &SpawnBurst_Instantaneous(Spawn Time=4, Use Spawn Probability=false, UseLoopCountLimit=false) {module=SpawnBurst_Instantaneous002}
   &SpawnBurst_Instantaneous(Spawn Time=4.6, Use Spawn Probability=false, UseLoopCountLimit=false) {module=SpawnBurst_Instantaneous001}
   &SpawnBurst_Instantaneous(Spawn Count=1, Spawn Time=4.8, Use Spawn Probability=false)
  ]
  particleSpawnScript [
   &InitializeParticle(
    Color=$User.GlowColor
    Lifetime=0.3
    Sprite Rotation Mode=::NewEnumerator2
    Sprite Size Mode=::NewEnumerator3
    Uniform Sprite Size=225
    UsePositionOffset=true
    Write Material Random=true
   )
   &DynamicMaterialParameters(
    Index 0 Param 1=&RandomRangeFloat(Minimum=0.4, Maximum=0.6)
    Index 0 Param 2=1
    Index 0 Param 3=1
    Index 0 Param 4=1
    Module.Param0WriteEnabled=true
    Module.Param1WriteEnabled=true
    Module.Param2WriteEnabled=true
    Module.Param3WriteEnabled=true
   )
   &CameraOffset(Camera Offset Amount=35)
  ]
  particleUpdateScript [
   &SkeletalMeshLocation(
    Bone / Socket Index=1
    Bone Sampling Mode=::NewEnumerator4
    MaskSpawnGroup=false
    Sampled Position Offset=<0 0 12>
    Sampled Position Offset Coordinate Space:ENiagaraSkelMeshTransforms=::NewEnumerator2
    Skeletal Mesh=$System.Skeletal Mesh
   )
   &ParticleState()
   &SolveForcesAndVelocity(Clamp Velocity=false, Limit Acceleration=false)
   &ScaleColor(
    Scale Alpha=&Multiply_Float(
     A=1
     B=&FloatFromCurve(FloatCurve=DataInterface() {propertyValues {Curve {keys [<at 0 0>,<at 0.2 1>,<at 0.8 1>,<at 1 0>]}}})
    )
    ScaleA=true
    ScaleRGB=true
   )
   &ScaleSpriteSize(Scale Factor=&Vector2DFromFloat(Value=&FloatFromCurve(FloatCurve=DataInterface() {propertyValues {Curve {keys [<at 0 1.5 RCIM_Cubic>,<at 1 0>]}}})))
   &FN_Light_Attributes(
    Color=$Particles.Color
    Create Color Attribute=true
    Create Exponent Attribute=true
    Create Position Attribute=true
    Create Radius Attribute=true
    Exponent=4
    Position=$System.SampleSkeletalMesh.SampledPosition
    Radius=200
    Scale Color=&FloatFromCurve(
     FloatCurve=DataInterface() {propertyValues {Curve {keys [
         {interpMode=RCIM_Cubic, tangentMode=RCTM_User, time=0, value=0}
         {interpMode=RCIM_Cubic, tangentMode=RCTM_User, time=0.5, value=1}
         {tangentMode=RCTM_User, time=1, value=0}
        ]}}}
     Scale Curve=5
    )
   )
  ]
  rendererClasses [&NiagaraSpriteRendererProperties, &NiagaraLightRendererProperties]
  renderers [
   {rendererClass=&NiagaraSpriteRendererProperties}
   {rendererClass=&NiagaraLightRendererProperties, rendererIndex=1}
  ]
 }
 GlowFlare001 {
  emitterUpdateScript [
   &EmitterState(
    Inactive Response=::NewEnumerator1
    Life Cycle Mode=::NewEnumerator1
    Loop Behavior=::NewEnumerator1
    Loop Delay=0.5
    Loop Duration=1
    UseLoopDelay=true
   )
   &SpawnBurst_Instantaneous(Spawn Count=1, Use Spawn Probability=false)
  ]
  particleSpawnScript [
   &InitializeParticle(
    Color=$User.GlowColor
    Lifetime=0.35
    Sprite Size Mode=::NewEnumerator3
    Uniform Sprite Size=130
    UsePositionOffset=true
    Write Material Random=true
   )
   &CameraOffset(Camera Offset Amount=50)
  ]
  particleUpdateScript [
   &SkeletalMeshLocation(Bone Sampling Mode=::NewEnumerator4, MaskSpawnGroup=false, Skeletal Mesh=$System.Skeletal Mesh)
   &ParticleState()
   &ScaleColor(
    Scale Alpha=&Multiply_Float(
     A=&FloatFromCurve(FloatCurve=DataInterface() {propertyValues {Curve {keys [{interpMode=RCIM_Cubic, leaveTangent=-1.8588, tangentMode=RCTM_Break, time=0, value=1}, <at 1 0>]}}})
     B=1
    )
    Scale RGB=<1 1 11>
    ScaleA=true
    ScaleRGB=true
   )
   &ScaleSpriteSize(Uniform Curve Sprite Scale=DataInterface() {propertyValues {Curve {keys [{tangentMode=RCTM_User, time=0, value=0}, <at 1 1>]}}})
  ]
  rendererClasses [&NiagaraSpriteRendererProperties]
  renderers [{rendererClass=&NiagaraSpriteRendererProperties}]
 }
 Glowbeam {
  emitterUpdateScript [
   &SampleSkeletalMesh(
    AdditionalYawPitchRoll=false
    Bone / Socket Index=1
    Bone Sampling Mode:ENiagaraBoneSamplingMode=::NewEnumerator4
    Skeletal Mesh:NiagaraDataInterfaceSkeletalMesh=$System.Skeletal Mesh
   )
   &EmitterState(Inactive Response=::NewEnumerator1, Life Cycle Mode=::NewEnumerator1, Loop Behavior=::NewEnumerator1, Loop Delay=1.1, UseLoopDelay=true)
   /Niagara/Modules/Beams/BeamEmitterSetup(Beam End:Vector3f=<0 0 130>, Beam Start:Vector3f=$Emitter.SampleSkeletalMesh.SampledPosition)
   &SpawnBurst_Instantaneous(Spawn Count=2, Use Spawn Probability=false, UseLoopCountLimit=false)
  ]
  particleSpawnScript [
   &PartitionParticles(
    Normalized Index In Partition Includes One=true
    Number Of Partitions=1
    Particles Per Partition=2
    Partition Mode:ENiagara_PartitionMode=::NewEnumerator1
   )
   &InitializeParticle(Color Mode=::NewEnumerator0, Lifetime=4.25, Position Mode=::NewEnumerator0, Write Material Random=true)
   /Niagara/Modules/Beams/SpawnBeam()
   /Niagara/Modules/Beams/BeamWidth(Beam Width=30)
   /SetVariables(Particles.NormalizedRibbonPos:NiagaraFloat=&OneMinusFloat(Float=$Particles.PartitionParticles.Normalized Index In Partition)) {isSetParameters=true, module=SetVariables_#2}
   /SetVariables(Particles.RibbonU0Override=$Particles.NormalizedRibbonPos) {isSetParameters=true, module=SetVariables_#3}
  ]
  particleUpdateScript [
   &ParticleState()
   /Niagara/Modules/Beams/UpdateBeam()
   &SolveForcesAndVelocity(Clamp Velocity=false, Limit Acceleration=false)
   &ScaleRibbonWidth(Ribbon Width Scale=&FloatFromCurve(FloatCurve=DataInterface() {propertyValues {Curve {keys [
         {arriveTangent=9.0725, interpMode=RCIM_Cubic, leaveTangent=9.0725, tangentMode=RCTM_User, time=0, value=0}
         {arriveTangent=0.1476, interpMode=RCIM_Cubic, leaveTangent=0.1476, tangentMode=RCTM_User, time=0.1408, value=0.9847}
         {arriveTangent=-0.5918, interpMode=RCIM_Cubic, leaveTangent=-0.5918, tangentMode=RCTM_User, time=0.7, value=1}
         <at 1 0 RCIM_Cubic>
        ]}}}))
   &DynamicMaterialParameters(
    Index 0 Param 1=&FloatFromCurve(FloatCurve=DataInterface() {propertyValues {Curve {keys [
         {arriveTangent=-9.1548, interpMode=RCIM_Cubic, leaveTangent=-9.1548, tangentMode=RCTM_User, time=0, value=0}
         {arriveTangent=-0.1162, interpMode=RCIM_Cubic, leaveTangent=-0.1162, tangentMode=RCTM_User, time=0.35, value=-1.6}
         {arriveTangent=5.8822, tangentMode=RCTM_User, time=1, value=0.5}
        ]}}})
    Index 0 Param 2=1
    Index 0 Param 3=&FloatFromCurve(FloatCurve=DataInterface() {propertyValues {Curve {keys [<at 0.8 1>,<at 1 0>]}}})
    Index 0 Param 4=&Multiply_Float(
     A=1
     B=&FloatFromCurve(
      FloatCurve=DataInterface() {propertyValues {Curve {keys [
          {arriveTangent=0.1868, interpMode=RCIM_Cubic, leaveTangent=0.1868, tangentMode=RCTM_User, time=0.1, value=-1}
          {arriveTangent=0.5209, interpMode=RCIM_Cubic, leaveTangent=0.5209, tangentMode=RCTM_User, time=0.2, value=1}
          {arriveTangent=-0.697, interpMode=RCIM_Cubic, leaveTangent=-0.697, tangentMode=RCTM_User, time=0.575, value=1}
          {arriveTangent=-0.6565, interpMode=RCIM_Cubic, leaveTangent=-0.6565, tangentMode=RCTM_User, time=0.65, value=-1}
         ]}}}
      CurveIndex=$System.NormalizedLoopAge
     )
    )
    Module.Param0WriteEnabled=true
    Module.Param1WriteEnabled=true
    Module.Param2WriteEnabled=true
    Module.Param3WriteEnabled=true
   )
   &ScaleColor(
    Scale Alpha=0.33
    Scale RGB=&Multiply_VectorByFloat(
     Vector=<1 1 1>
     Float=&FloatFromCurve(
      FloatCurve=DataInterface() {propertyValues {Curve {keys [
          {arriveTangent=9.0725, interpMode=RCIM_Cubic, leaveTangent=9.0725, tangentMode=RCTM_User, time=0, value=0}
          {arriveTangent=-0.0378, interpMode=RCIM_Cubic, leaveTangent=-0.0378, tangentMode=RCTM_User, time=0.1408, value=0.9847}
          {arriveTangent=-0.5683, interpMode=RCIM_Cubic, leaveTangent=-0.5683, tangentMode=RCTM_User, time=0.5, value=0.9}
          {arriveTangent=-0.0405, interpMode=RCIM_Cubic, leaveTangent=-0.0405, tangentMode=RCTM_User, time=0.75, value=0}
         ]}}}
      CurveIndex=$System.NormalizedLoopAge
     )
    )
    ScaleA=true
    ScaleRGB=true
   )
  ]
  rendererClasses [&NiagaraRibbonRendererProperties]
  renderers [{rendererClass=&NiagaraRibbonRendererProperties}]
 }
 Glowbeam_Soft {
  emitterUpdateScript [
   &SampleSkeletalMesh(
    AdditionalYawPitchRoll=false
    Bone / Socket Index=1
    Bone Sampling Mode:ENiagaraBoneSamplingMode=::NewEnumerator4
    Skeletal Mesh:NiagaraDataInterfaceSkeletalMesh=$System.Skeletal Mesh
   )
   &EmitterState(Inactive Response=::NewEnumerator1, Life Cycle Mode=::NewEnumerator1, Loop Behavior=::NewEnumerator1, Loop Duration=6, UseLoopDelay=true)
   &BeamEmitterSetup(Beam End:Vector3f=<0 0 150>, Beam Start:Vector3f=$Emitter.SampleSkeletalMesh.SampledPosition)
   &SpawnBurst_Instantaneous(Spawn Count=2, Use Spawn Probability=false, UseLoopCountLimit=false)
  ]
  particleSpawnScript [
   &PartitionParticles(
    Normalized Index In Partition Includes One=true
    Number Of Partitions=1
    Particles Per Partition=2
    Partition Mode:ENiagara_PartitionMode=::NewEnumerator1
   )
   &InitializeParticle(Color=<rgba 1.5 0.6102 0.3999 1>, Lifetime=5.4, Position Mode=::NewEnumerator0, Write Material Random=true)
   &SpawnBeam()
   /SetVariables(Particles.NormalizedRibbonPos:NiagaraFloat=&OneMinusFloat(Float=$Particles.PartitionParticles.Normalized Index In Partition)) {isSetParameters=true, module=SetVariables_#2}
   /SetVariables(Particles.RibbonU0Override=$Particles.NormalizedRibbonPos) {isSetParameters=true, module=SetVariables_#3}
   &BeamWidth(Beam Width:NiagaraFloat=&Lerp_Float(
     A=80
     B=165
     Alpha=&FloatFromCurve(
      FloatCurve=DataInterface() {propertyValues {Curve {keys [
          {arriveTangent=-1.621, interpMode=RCIM_Cubic, leaveTangent=-1.621, tangentMode=RCTM_User, time=0, value=1}
          {interpMode=RCIM_Cubic, tangentMode=RCTM_User, time=1, value=0}
         ]}}}
      CurveIndex=$Particles.NormalizedRibbonPos
     )
    ))
  ]
  particleUpdateScript [
   &ParticleState()
   &UpdateBeam()
   &SolveForcesAndVelocity(Clamp Velocity=false, Limit Acceleration=false)
   &ScaleRibbonWidth(Ribbon Width Scale=&FloatFromCurve(FloatCurve=DataInterface() {propertyValues {Curve {keys [
         {arriveTangent=9.0725, interpMode=RCIM_Cubic, leaveTangent=9.0725, tangentMode=RCTM_User, time=0, value=0}
         {arriveTangent=0.1476, leaveTangent=0.1476, time=0.1408, value=0.9847}
         {arriveTangent=-0.5918, interpMode=RCIM_Cubic, leaveTangent=-0.5918, tangentMode=RCTM_User, time=0.8, value=1}
         {arriveTangent=-5, interpMode=RCIM_Cubic, leaveTangent=-5, time=0.9, value=0.725}
         {interpMode=RCIM_Cubic, leaveTangent=0.0233, time=1, value=0}
        ]}}}))
   &DynamicMaterialParameters(
    Index 0 Param 1=&FloatFromCurve(FloatCurve=DataInterface() {propertyValues {Curve {keys [
         {arriveTangent=-9.1548, interpMode=RCIM_Cubic, leaveTangent=-9.1548, tangentMode=RCTM_User, time=0, value=0}
         {arriveTangent=-0.1162, interpMode=RCIM_Cubic, leaveTangent=-0.1162, tangentMode=RCTM_User, time=0.4, value=-1.6}
         {arriveTangent=5.8822, tangentMode=RCTM_User, time=1, value=0.5}
        ]}}})
    Index 0 Param 2=1
    Index 0 Param 3=&FloatFromCurve(FloatCurve=DataInterface() {propertyValues {Curve {keys [<at 0.85 1>,<at 1 0>]}}})
    Index 0 Param 4=&Multiply_Float(
     A=1
     B=&FloatFromCurve(
      FloatCurve=DataInterface() {propertyValues {Curve {keys [
          {arriveTangent=10.8222, interpMode=RCIM_Cubic, leaveTangent=10.8222, tangentMode=RCTM_User, time=0.075, value=-1}
          {arriveTangent=1.4866, interpMode=RCIM_Cubic, leaveTangent=1.4866, tangentMode=RCTM_User, time=0.3, value=1}
          {arriveTangent=-1.7929, interpMode=RCIM_Cubic, leaveTangent=-1.7929, tangentMode=RCTM_User, time=0.575, value=1}
          {arriveTangent=-1.7566, interpMode=RCIM_Cubic, leaveTangent=-1.7566, tangentMode=RCTM_User, time=0.75, value=-1}
         ]}}}
      CurveIndex=$System.NormalizedLoopAge
     )
    )
    Module.Param0WriteEnabled=true
    Module.Param1WriteEnabled=true
    Module.Param2WriteEnabled=true
    Module.Param3WriteEnabled=true
   )
   &ScaleColor(
    Scale Alpha=0.5
    Scale RGB=&Multiply_VectorByFloat(
     Vector=<1 1 1>
     Float=&FloatFromCurve(
      FloatCurve=DataInterface() {propertyValues {Curve {keys [
          {arriveTangent=9.0725, interpMode=RCIM_Cubic, leaveTangent=9.0725, tangentMode=RCTM_User, time=0, value=0}
          {arriveTangent=-0.0378, leaveTangent=-0.0378, time=0.1408, value=0.9847}
          {arriveTangent=-0.5683, interpMode=RCIM_Cubic, leaveTangent=-0.5683, tangentMode=RCTM_User, time=0.55, value=1}
          {arriveTangent=-0.0405, interpMode=RCIM_Cubic, leaveTangent=-0.0405, tangentMode=RCTM_User, time=0.8, value=0}
         ]}}}
      CurveIndex=$System.NormalizedLoopAge
     )
    )
    ScaleA=true
    ScaleRGB=true
   )
  ]
  rendererClasses [&NiagaraRibbonRendererProperties]
  renderers [{rendererClass=&NiagaraRibbonRendererProperties}]
 }
 LightShaftsBurst {
  emitterUpdateScript [
   &EmitterState(Inactive Response=::NewEnumerator1, Life Cycle Mode=::NewEnumerator1, Loop Behavior=::NewEnumerator1, Loop Delay=0.5, UseLoopDelay=true)
   &SpawnBurst_Instantaneous(Use Spawn Probability=false, UseLoopCountLimit=false)
  ]
  particleSpawnScript [
   &InitializeParticle(
    Color=<rgba 2 0.9604 0 1>
    Lifetime=5
    Mesh Scale Mode=::NewEnumerator3
    Mesh Uniform Scale=0.05
    Position Offset=<0 0 -5>
    UsePositionOffset=true
    Write Material Random=true
   )
   &InitialMeshOrientation(Orientation Axis=<0 0 -1>, Orientation Vector=<0 0 1>, Use Orientation Vector=true, Use Rotation Vector=false)
  ]
  particleUpdateScript [
   &ParticleState()
   &ScaleMeshSize(Scale Factor=&Multiply_VectorByFloat(
     Vector=<1 1 1>
     Float=&FloatFromCurve(FloatCurve=DataInterface() {propertyValues {Curve {keys [{tangentMode=RCTM_User, time=0.2, value=1}, {tangentMode=RCTM_User, time=0.3, value=0}]}}})
    ))
   &DynamicMaterialParameters(
    Index 0 Param 1=&FloatFromCurve(
     FloatCurve=DataInterface() {propertyValues {Curve {keys [{interpMode=RCIM_Cubic, tangentMode=RCTM_User, time=0, value=1}, {interpMode=RCIM_Cubic, tangentMode=RCTM_User, time=1, value=0.25}]}}}
     Scale Curve=0.2
    )
    Index 0 Param 2=&Lerp_Float(
     A=0.05
     B=0.1
     Alpha=&FloatFromCurve(FloatCurve=DataInterface() {propertyValues {Curve {keys [
          {arriveTangent=4.7763, interpMode=RCIM_Cubic, leaveTangent=4.7763, tangentMode=RCTM_User, time=0, value=0}
          {arriveTangent=0.4097, interpMode=RCIM_Cubic, leaveTangent=0.4097, tangentMode=RCTM_User, time=0.3, value=0.9}
          <at 1 1>
         ]}}})
    )
    Index 0 Param 3=&FloatFromCurve(
     FloatCurve=DataInterface() {propertyValues {Curve {keys [{arriveTangent=1.5696, interpMode=RCIM_Cubic, leaveTangent=1.5696, tangentMode=RCTM_User, time=0, value=0}, <at 0.4 1 RCIM_Cubic>]}}}
     CurveIndex=$Particles.Age
     Scale Curve=0.02
    )
    Index 0 Param 4=15
    Module.Param0WriteEnabled=true
    Module.Param1WriteEnabled=true
    Module.Param2WriteEnabled=true
    Module.Param3WriteEnabled=true
   )
   &ScaleColor(
    Scale RGB=&VectorFromFloat(Value=&FloatFromCurve(
      FloatCurve=DataInterface() {propertyValues {Curve {keys [
          {arriveTangent=9.0725, interpMode=RCIM_Cubic, leaveTangent=9.0725, tangentMode=RCTM_User, time=0, value=0}
          {arriveTangent=-2.5249, interpMode=RCIM_Cubic, leaveTangent=-2.5249, tangentMode=RCTM_User, time=0.1408, value=0.9847}
          {arriveTangent=-0.0405, interpMode=RCIM_Cubic, leaveTangent=-0.0405, tangentMode=RCTM_User, time=1, value=0}
         ]}}}
      CurveIndex=$Particles.Age
      Scale Curve=5
     ))
    ScaleA=true
    ScaleRGB=true
   ) {module=ScaleColor001}
  ]
  rendererClasses [&NiagaraMeshRendererProperties]
  renderers [{rendererClass=&NiagaraMeshRendererProperties, rendererIndex=0}]
 }
 Shine {
  emitterUpdateScript [
   &EmitterState(
    Inactive Response=::NewEnumerator1
    Life Cycle Mode=::NewEnumerator1
    Loop Behavior=::NewEnumerator1
    Loop Delay=0.6
    Loop Duration=3.75
    UseLoopDelay=true
   )
   &SpawnRate(SpawnRate=33, Use Spawn Probability=false)
  ]
  particleSpawnScript [
   &InitializeParticle(
    Color=<rgba 2 2 2 1>
    Lifetime Max=0.7
    Lifetime Min=0.5
    Lifetime Mode=::NewEnumerator1
    Sprite Rotation Mode=::NewEnumerator2
    Sprite Size Max=<20 20>
    Sprite Size Min=<15 15>
    Sprite Size Mode=::NewEnumerator2
    UsePositionOffset=false
    Write Material Random=true
   )
   &SubUVAnimation(End Frame=3, SubUV Animation Mode:ENiagara_SubUVLookupMode=::NewEnumerator2)
  ]
  particleUpdateScript [
   &SkeletalMeshLocation(Bone Sampling Mode=::NewEnumerator4, MaskSpawnGroup=false, Skeletal Mesh=$System.Skeletal Mesh)
   &ParticleState()
   &ScaleSpriteSize(Scale Factor=&MakeVector2D(
     X=&FloatFromCurve(FloatCurve=DataInterface() {propertyValues {Curve {keys [
          {interpMode=RCIM_Cubic, tangentMode=RCTM_User, time=0, value=0}
          {interpMode=RCIM_Cubic, tangentMode=RCTM_User, time=0.2, value=1}
          {tangentMode=RCTM_User, time=1, value=0.5}
         ]}}})
     Y=&FloatFromCurve(FloatCurve=DataInterface() {propertyValues {Curve {keys [
          <at 0 0 RCIM_Cubic>
          {arriveTangent=1.5, interpMode=RCIM_Cubic, leaveTangent=1.5, time=0.2, value=0.9}
          {tangentMode=RCTM_User, time=1, value=1.5}
         ]}}})
    ))
   &ScaleColor(
    Scale Alpha=&FloatFromCurve(
     FloatCurve=DataInterface() {propertyValues {Curve {keys [{interpMode=RCIM_Cubic, tangentMode=RCTM_User, time=0, value=1}, {interpMode=RCIM_Cubic, tangentMode=RCTM_User, time=1, value=0}]}}}
     Scale Curve=&FloatFromCurve(
      FloatCurve=DataInterface() {propertyValues {Curve {keys [
          {interpMode=RCIM_Cubic, tangentMode=RCTM_User, time=0, value=0}
          {arriveTangent=1.25, leaveTangent=1.25, time=0.2, value=1}
          {interpMode=RCIM_Cubic, tangentMode=RCTM_User, time=0.8, value=1}
          {tangentMode=RCTM_User, time=1, value=0}
         ]}}}
      CurveIndex=$System.NormalizedLoopAge
     )
    )
    Scale RGB=&MakeVectorFromLinearColorRGB(Color=$User.GlowColor)
    ScaleA=true
    ScaleRGB=true
   )
   /SetVariables(Particles.SpriteSize=&Multiply_Vector2D_ByFloat(
     Float=&FloatFromCurve(
      FloatCurve=DataInterface() {propertyValues {Curve {keys [
          {arriveTangent=4.3493, interpMode=RCIM_Cubic, leaveTangent=4.3493, tangentMode=RCTM_User, time=0, value=0}
          {arriveTangent=1.25, leaveTangent=1.25, time=0.2, value=1}
          {interpMode=RCIM_Cubic, tangentMode=RCTM_User, time=0.8, value=1}
          {arriveTangent=-2.788, tangentMode=RCTM_User, time=1, value=0}
         ]}}}
      CurveIndex=$System.NormalizedLoopAge
     )
     Vector2D=$Particles.SpriteSize
    )) {isSetParameters=true}
  ]
  rendererClasses [&NiagaraSpriteRendererProperties]
  renderers [{rendererClass=&NiagaraSpriteRendererProperties}]
 }
 SmokeBlackDetail {
  emitterUpdateScript [
   &EmitterState(
    Inactive Response=::NewEnumerator2
    Life Cycle Mode=::NewEnumerator1
    Loop Behavior=::NewEnumerator1
    Loop Delay=0.5
    Loop Duration=4
    UseLoopDelay=true
   )
   &SpawnRate(SpawnRate=5, Use Spawn Probability=false)
   &SpawnBurst_Instantaneous(Spawn Count=3, Use Spawn Probability=false, UseLoopCountLimit=false)
  ]
  particleSpawnScript [
   &InitializeParticle(
    Lifetime Max=1.5
    Lifetime Mode=::NewEnumerator1
    Sprite Rotation Mode=::NewEnumerator2
    Sprite Size Max=<45 45>
    Sprite Size Min=<35 35>
    Sprite Size Mode=::NewEnumerator2
    UsePositionOffset=false
    Write Material Random=true
   )
   &CameraOffset(Camera Offset Amount=5)
   &AddVelocity(Velocity=&RandomRangeVector(Minimum=<0 0 10>, Maximum=<0 0 25>))
  ]
  particleUpdateScript [
   &ParticleState()
   &SpriteRotationRate(Rotation Rate=&RandomRangeFloat(Minimum=-50, Maximum=50))
   &ScaleSpriteSize(
    Scale Sprite Size Mode=::NewEnumerator0
    Uniform Scale Factor=&FloatFromCurve(
     FloatCurve=DataInterface() {propertyValues {Curve {keys [
         {arriveTangent=4.1785, interpMode=RCIM_Cubic, leaveTangent=4.1785, tangentMode=RCTM_User, time=0, value=0}
         {arriveTangent=1.1179, interpMode=RCIM_Cubic, leaveTangent=1.1179, tangentMode=RCTM_User, time=0.2, value=0.675}
         {arriveTangent=0.0224, interpMode=RCIM_Cubic, leaveTangent=0.0224, tangentMode=RCTM_User, time=1, value=1}
        ]}}}
     Scale Curve=1.3
    )
   )
   &SubUVAnimation(
    End Frame Range Override=30
    "Sprite Renderer "=DataInterface() {propertyValues {SpriteRenderer=&self:SmokeBlackDetail_6.NiagaraSpriteRendererProperties_0}}
    Start Frame Range Override=&RandomRangeInt(Minimum=0, Maximum=5)
    UseEndFrame=true
    UseStartFrame=true
   )
   &ScaleColor(
    Scale Alpha=&FloatFromCurve(FloatCurve=DataInterface() {propertyValues {Curve {keys [
         {arriveTangent=9.0725, interpMode=RCIM_Cubic, leaveTangent=9.0725, tangentMode=RCTM_User, time=0, value=0}
         {arriveTangent=-2.5249, interpMode=RCIM_Cubic, leaveTangent=-2.5249, tangentMode=RCTM_User, time=0.1408, value=0.9847}
         {arriveTangent=-0.0405, interpMode=RCIM_Cubic, leaveTangent=-0.0405, tangentMode=RCTM_User, time=1, value=0}
        ]}}})
    Scale RGB=&MakeVectorFromLinearColorRGB(Color=&ColorFromCurve(ColorCurve=DataInterface() {propertyValues {
        AlphaCurve {keys [<at 0 1 RCIM_Cubic>,<at 1 0 RCIM_Cubic>]}
        BlueCurve {keys [<at 0 0 RCIM_Cubic>,<at 0.5105 0 RCIM_Cubic>]}
        GreenCurve {keys [<at 0 0.8547 RCIM_Cubic>,<at 0.5105 0.3126 RCIM_Cubic>]}
        RedCurve {keys [<at 0 1 RCIM_Cubic>,<at 0.5105 1 RCIM_Cubic>]}
       }}))
    ScaleA=true
    ScaleRGB=true
   )
   &Drag(Drag=3, Use Linear Drag=true, Use Rotational Drag=true)
   &SolveForcesAndVelocity(Clamp Velocity=false, Limit Acceleration=false)
  ]
  rendererClasses [&NiagaraSpriteRendererProperties]
  renderers [{rendererClass=&NiagaraSpriteRendererProperties}]
 }
 SmokeBlackDetail001 {
  emitterUpdateScript [
   &EmitterState(
    Inactive Response=::NewEnumerator2
    Life Cycle Mode=::NewEnumerator1
    Loop Behavior=::NewEnumerator1
    Loop Delay=4.85
    Loop Duration=2
    UseLoopDelay=true
   )
   &SpawnBurst_Instantaneous(Spawn Count=3, Use Spawn Probability=false, UseLoopCountLimit=false)
  ]
  particleSpawnScript [
   &InitializeParticle(
    Lifetime Max=1.5
    Lifetime Mode=::NewEnumerator1
    Sprite Rotation Mode=::NewEnumerator2
    Sprite Size Max=<55 55>
    Sprite Size Min=<45 45>
    Sprite Size Mode=::NewEnumerator2
    UsePositionOffset=true
    Write Material Random=true
   )
   &CameraOffset(Camera Offset Amount=5)
   &AddVelocity(Velocity=&RandomRangeVector(Minimum=<0 0 30>, Maximum=<0 0 45>))
  ]
  particleUpdateScript [
   &ParticleState()
   &SpriteRotationRate(Rotation Rate=&RandomRangeFloat(Minimum=-50, Maximum=50))
   &ScaleSpriteSize(
    Scale Sprite Size Mode=::NewEnumerator0
    Uniform Scale Factor=&FloatFromCurve(
     FloatCurve=DataInterface() {propertyValues {Curve {keys [
         {arriveTangent=4.1785, interpMode=RCIM_Cubic, leaveTangent=4.1785, tangentMode=RCTM_User, time=0, value=0}
         {arriveTangent=1.1179, interpMode=RCIM_Cubic, leaveTangent=1.1179, tangentMode=RCTM_User, time=0.2, value=0.675}
         {arriveTangent=0.0224, interpMode=RCIM_Cubic, leaveTangent=0.0224, tangentMode=RCTM_User, time=1, value=1}
        ]}}}
     Scale Curve=1.3
    )
   )
   &SubUVAnimation(
    End Frame Range Override=30
    "Sprite Renderer "=DataInterface() {propertyValues {SpriteRenderer=&self:SmokeBlackDetail001_8.NiagaraSpriteRendererProperties_0}}
    Start Frame Range Override=&RandomRangeInt(Minimum=0, Maximum=5)
    UseEndFrame=true
    UseStartFrame=true
   )
   &ScaleColor(
    Scale Alpha=&FloatFromCurve(FloatCurve=DataInterface() {propertyValues {Curve {keys [
         {arriveTangent=9.0725, interpMode=RCIM_Cubic, leaveTangent=9.0725, tangentMode=RCTM_User, time=0, value=0}
         {arriveTangent=-2.5249, interpMode=RCIM_Cubic, leaveTangent=-2.5249, tangentMode=RCTM_User, time=0.1408, value=0.9847}
         {arriveTangent=-0.0405, interpMode=RCIM_Cubic, leaveTangent=-0.0405, tangentMode=RCTM_User, time=1, value=0}
        ]}}})
    Scale RGB=&MakeVectorFromLinearColorRGB(Color=&ColorFromCurve(ColorCurve=DataInterface() {propertyValues {
        AlphaCurve {keys [<at 0 1 RCIM_Cubic>,<at 1 0 RCIM_Cubic>]}
        BlueCurve {keys [<at 0 0 RCIM_Cubic>,<at 0.5105 0 RCIM_Cubic>]}
        GreenCurve {keys [<at 0 0.8547 RCIM_Cubic>,<at 0.5105 0.3126 RCIM_Cubic>]}
        RedCurve {keys [<at 0 1 RCIM_Cubic>,<at 0.5105 1 RCIM_Cubic>]}
       }}))
    ScaleA=true
    ScaleRGB=true
   )
   &Drag(Drag=3, Use Linear Drag=true, Use Rotational Drag=true)
   &InheritSourceMovement(
    Applied Position Scale=&Multiply_VectorByFloat(
     Vector=<1 1 1>
     Float=&FloatFromCurve(FloatCurve=DataInterface() {propertyValues {Curve {keys [{interpMode=RCIM_Cubic, tangentMode=RCTM_User, time=0.25, value=1}, {interpMode=RCIM_Cubic, tangentMode=RCTM_User, time=1, value=0}]}}})
    )
    ApplyForce=false
    ApplyPosition=true
    ApplyPositionFalloff=false
   )
   &SolveForcesAndVelocity(Clamp Velocity=false, Limit Acceleration=false)
  ]
  rendererClasses [&NiagaraSpriteRendererProperties]
  renderers [{rendererClass=&NiagaraSpriteRendererProperties}]
 }
 SparkDown {
  simTarget=GPUComputeSim
  emitterUpdateScript [
   &EmitterState(
    Enable Distance Culling=true
    Inactive Response=::NewEnumerator2
    Life Cycle Mode=::NewEnumerator1
    Loop Behavior=::NewEnumerator1
    Loop Delay=2
    Loop Duration=2
    MaxDistance=1750
    Module.bUseMaxDistance=true
    Module.bUseMinDistance=true
    Scalability Mode=::NewEnumerator1
    UseLoopDelay=true
   )
   &SpawnRate(SpawnRate=75, Use Spawn Probability=false)
  ]
  particleSpawnScript [
   &InitializeParticle(
    Color Mode=::NewEnumerator0
    Lifetime Max=1.25
    Lifetime Min=0.5
    Lifetime Mode=::NewEnumerator1
    Sprite Size Max=<1.25 1.25>
    Sprite Size Min=<0.75 0.75>
    Sprite Size Mode=::NewEnumerator2
    UsePositionOffset=false
    Write Material Random=true
   )
   &ShapeLocation(
    Cylinder Height=30
    Cylinder Radius=35
    MaskSpawnGroup=false
    Offset=<0 0 75>
    Offset Mode=::NewEnumerator0
    Override Seed=false
    Shape Origin=$System.SampleSkeletalMesh.SampledPosition
    Shape Primitive=::NewEnumerator1
    Surface Only=false
    UseOwnerScale=true
   )
   &ApplyInitialForces()
  ]
  particleUpdateScript [
   &ParticleState()
   &InheritSourceMovement(
    Applied Position Scale=&Multiply_VectorByFloat(Vector=<1 1 1>, Float=1)
    ApplyForce=false
    ApplyPosition=true
    ApplyPositionFalloff=false
   )
   &CameraOffset(Camera Offset Amount=10)
   &ScaleColor(
    Scale RGB=&MakeVectorFromLinearColorRGB(Color=&ColorFromCurve(ColorCurve=DataInterface() {propertyValues {
        AlphaCurve {keys [<at 0 1 RCIM_Cubic>,<at 1 0 RCIM_Cubic>]}
        BlueCurve {keys [<at 0 0 RCIM_Cubic>,<at 0.2652 0>,<at 1 0 RCIM_Cubic>]}
        GreenCurve {keys [<at 0 3.0058 RCIM_Cubic>,<at 0.2652 0.8855>,<at 1 0.3858 RCIM_Cubic>]}
        RedCurve {keys [<at 0 5 RCIM_Cubic>,<at 0.2652 2>,<at 1 1 RCIM_Cubic>]}
       }}))
    ScaleA=true
    ScaleRGB=true
   )
   &ScaleSpriteSize(Uniform Curve Sprite Scale=DataInterface() {propertyValues {Curve {keys [
        {interpMode=RCIM_Cubic, tangentMode=RCTM_User, time=0, value=0}
        {interpMode=RCIM_Cubic, tangentMode=RCTM_User, time=0.5, value=1}
        {tangentMode=RCTM_User, time=1, value=0}
       ]}}})
   &Drag(Drag=&RandomRangeFloat(Minimum=1, Maximum=2), Use Linear Drag=true, Use Rotational Drag=true)
   &VortexForce(
    Origin Pull Amount=500
    UseOriginOffset=true
    Vortex Axis=<0 0 -1>
    Vortex Force Amount=500
    Vortex Origin:NiagaraPosition=$System.SampleSkeletalMesh.SampledPosition
   )
   &PointAttractionForce(
    Attraction Radius=250
    Attraction Strength=&RandomRangeFloat(Minimum=25, Maximum=35)
    Attractor Position:NiagaraPosition=$System.SampleSkeletalMesh.SampledPosition
    Kill Within Radius=false
    Module.Use Falloff=true
    UseAttractorOffset=false
   )
   &SolveForcesAndVelocity(Clamp Velocity=false, Limit Acceleration=false)
   &ScaleSpriteSizeBySpeed(Max Scale Factor=<1 4>, Min Scale Factor=<1 0>, Velocity Threshold=150)
  ]
  rendererClasses [&NiagaraSpriteRendererProperties]
  renderers [{rendererClass=&NiagaraSpriteRendererProperties}]
 }
 SparkUp {
  simTarget=GPUComputeSim
  emitterUpdateScript [
   &EmitterState(
    Enable Distance Culling=true
    Inactive Response=::NewEnumerator2
    Life Cycle Mode=::NewEnumerator1
    Loop Behavior=::NewEnumerator1
    Loop Delay=1
    Loop Duration=1
    MaxDistance=1750
    Module.bUseMaxDistance=true
    Module.bUseMinDistance=true
    Scalability Mode=::NewEnumerator1
    UseLoopDelay=true
   )
   &SpawnRate(SpawnRate=100, Use Spawn Probability=false)
  ]
  particleSpawnScript [
   &InitializeParticle(
    Color Mode=::NewEnumerator0
    Lifetime Max=1.25
    Lifetime Min=0.5
    Lifetime Mode=::NewEnumerator1
    Sprite Size Max=<1.25 1.25>
    Sprite Size Min=<0.75 0.75>
    Sprite Size Mode=::NewEnumerator2
    UsePositionOffset=false
    Write Material Random=true
   )
   &AddVelocity(
    Velocity=&RandomRangeVector(Minimum=<0 0 75>, Maximum=<0 0 200>)
    Velocity Speed Scale=&FloatFromCurve(
     FloatCurve=DataInterface() {propertyValues {Curve {keys [<at 0.5 1 RCIM_Cubic>, {arriveTangent=-1.3067, interpMode=RCIM_Cubic, leaveTangent=-1.3067, tangentMode=RCTM_User, time=1, value=0.3}]}}}
     CurveIndex=$Emitter.NormalizedLoopAge
    )
   )
   &ShapeLocation(
    Cylinder Height=10
    Cylinder Radius=10
    MaskSpawnGroup=false
    Offset Mode=::NewEnumerator0
    Override Seed=false
    Shape Origin=$System.SampleSkeletalMesh.SampledPosition
    Shape Primitive=::NewEnumerator1
    Surface Only=false
    UseOwnerScale=true
   )
   &ApplyInitialForces()
  ]
  particleUpdateScript [
   &ParticleState()
   &InheritSourceMovement(
    Applied Position Scale=&Multiply_VectorByFloat(Vector=<1 1 1>, Float=1)
    ApplyForce=false
    ApplyPosition=true
    ApplyPositionFalloff=false
   )
   &ScaleColor(
    Scale RGB=&MakeVectorFromLinearColorRGB(Color=&ColorFromCurve(ColorCurve=DataInterface() {propertyValues {
        AlphaCurve {keys [<at 0 1 RCIM_Cubic>,<at 1 0 RCIM_Cubic>]}
        BlueCurve {keys [<at 0 0 RCIM_Cubic>,<at 0.2652 0>,<at 1 0 RCIM_Cubic>]}
        GreenCurve {keys [<at 0 3.0058 RCIM_Cubic>,<at 0.2652 0.8855>,<at 1 0.3858 RCIM_Cubic>]}
        RedCurve {keys [<at 0 5 RCIM_Cubic>,<at 0.2652 2>,<at 1 1 RCIM_Cubic>]}
       }}))
    ScaleA=true
    ScaleRGB=true
   )
   &ScaleSpriteSize(Uniform Curve Sprite Scale=DataInterface() {propertyValues {Curve {keys [
        {interpMode=RCIM_Cubic, tangentMode=RCTM_User, time=0, value=0}
        {interpMode=RCIM_Cubic, tangentMode=RCTM_User, time=0.5, value=1}
        {tangentMode=RCTM_User, time=1, value=0}
       ]}}})
   &Drag(Drag=&RandomRangeFloat(Minimum=1, Maximum=2), Use Linear Drag=true, Use Rotational Drag=true)
   &VortexForce(
    Origin Pull Amount=500
    UseOriginOffset=true
    Vortex Force Amount=500
    Vortex Origin:NiagaraPosition=$System.SampleSkeletalMesh.SampledPosition
   )
   &SolveForcesAndVelocity(Clamp Velocity=false, Limit Acceleration=false)
   &ScaleSpriteSizeBySpeed(Max Scale Factor=<1 4>, Min Scale Factor=<1 0>, Velocity Threshold=150)
  ]
  rendererClasses [&NiagaraSpriteRendererProperties]
  renderers [{rendererClass=&NiagaraSpriteRendererProperties}]
 }
 SparksOpen {
  simTarget=GPUComputeSim
  emitterUpdateScript [
   &EmitterState(
    Enable Distance Culling=true
    Inactive Response=::NewEnumerator1
    Life Cycle Mode=::NewEnumerator1
    Loop Behavior=::NewEnumerator1
    Loop Delay=0.5
    Loop Duration=6
    MaxDistance=1750
    Module.bUseMaxDistance=true
    Module.bUseMinDistance=true
    Scalability Mode=::NewEnumerator1
    UseLoopDelay=true
   )
   &SpawnBurst_Instantaneous(Spawn Count=25, Use Spawn Probability=false, UseLoopCountLimit=false) {module=SpawnBurst_Instantaneous003}
   &SpawnBurst_Instantaneous(Spawn Count=10, Spawn Time=3.8, Use Spawn Probability=false, UseLoopCountLimit=false) {module=SpawnBurst_Instantaneous002}
   &SpawnBurst_Instantaneous(Spawn Count=10, Spawn Time=4.4, Use Spawn Probability=false, UseLoopCountLimit=false) {module=SpawnBurst_Instantaneous001}
   &SpawnBurst_Instantaneous(Spawn Count=10, Spawn Time=4.6, Use Spawn Probability=false, UseLoopCountLimit=false)
  ]
  particleSpawnScript [
   &InitializeParticle(
    Color Mode=::NewEnumerator0
    Lifetime Max=1.5
    Lifetime Min=0.5
    Lifetime Mode=::NewEnumerator1
    Sprite Size Max=<1 1>
    Sprite Size Min=<0.5 0.5>
    Sprite Size Mode=::NewEnumerator2
    UsePositionOffset=false
    Write Material Random=true
   )
   &SkeletalMeshLocation(Bone Sampling Mode=::NewEnumerator4, MaskSpawnGroup=false, Skeletal Mesh=$System.Skeletal Mesh)
   &AddVelocity(
    Cone Angle=65
    Cone Axis=<0 0 1>
    Module.UseConeFalloff=false
    Module.UseVelDistribution=false
    Override Seed=false
    Velocity Mode=::NewEnumerator2
    Velocity Speed=&RandomRangeFloat(Minimum=25, Maximum=80)
   )
   &ShapeLocation(Hemisphere Distribution=<1 0.5>, MaskSpawnGroup=false, Override Seed=false, Sphere Radius=7, UseOwnerScale=true)
   &ApplyInitialForces()
  ]
  particleUpdateScript [
   &ParticleState()
   &ScaleColor(
    Scale RGB=&MakeVectorFromLinearColorRGB(Color=&ColorFromCurve(ColorCurve=DataInterface() {propertyValues {
        AlphaCurve {keys [<at 0 1 RCIM_Cubic>,<at 1 0 RCIM_Cubic>]}
        BlueCurve {keys [<at 0 0 RCIM_Cubic>,<at 0.2652 0>,<at 1 0 RCIM_Cubic>]}
        GreenCurve {keys [<at 0 3.0058 RCIM_Cubic>,<at 0.2652 0.8855>,<at 1 0.3858 RCIM_Cubic>]}
        RedCurve {keys [<at 0 5 RCIM_Cubic>,<at 0.2652 2>,<at 1 1 RCIM_Cubic>]}
       }}))
    ScaleA=true
    ScaleRGB=true
   )
   &ScaleSpriteSize(Uniform Curve Sprite Scale=DataInterface() {propertyValues {Curve {keys [
        {interpMode=RCIM_Cubic, tangentMode=RCTM_User, time=0, value=0}
        {arriveTangent=1.7917, interpMode=RCIM_Cubic, leaveTangent=1.7917, time=0.2, value=1}
        {interpMode=RCIM_Cubic, tangentMode=RCTM_User, time=0.6, value=1.075}
        {tangentMode=RCTM_User, time=1, value=0}
       ]}}})
   &Drag(Drag=&RandomRangeFloat(Minimum=1, Maximum=2), Use Linear Drag=true, Use Rotational Drag=true)
   &InheritSourceMovement(
    Applied Position Scale=&Multiply_VectorByFloat(
     Vector=<1 1 1>
     Float=&FloatFromCurve(FloatCurve=DataInterface() {propertyValues {Curve {keys [{interpMode=RCIM_Cubic, tangentMode=RCTM_User, time=0.25, value=1}, {interpMode=RCIM_Cubic, tangentMode=RCTM_User, time=1, value=0}]}}})
    )
    ApplyForce=false
    ApplyPosition=true
    ApplyPositionFalloff=false
   )
   /Niagara/Modules/Update/Velocity/VortexVelocity(
    Module.Use Influence Falloff=false
    Velocity Amount:NiagaraFloat=&RandomRangeFloat(Minimum=1000, Maximum=3000)
    Vortex Origin:NiagaraPosition=$Particles.SkeletalMeshLocation.SampledPosition
   )
   &SolveForcesAndVelocity(Clamp Velocity=false, Limit Acceleration=false)
   &ScaleSpriteSizeBySpeed(Max Scale Factor=<1 4>, Min Scale Factor=<1 0>, Velocity Threshold=50)
  ]
  rendererClasses [&NiagaraSpriteRendererProperties]
  renderers [{rendererClass=&NiagaraSpriteRendererProperties}]
 }
 Spirals_Ribbon_Down {
  emitterUpdateScript [
   &EmitterState(
    Delay First Loop Only=true
    Enable Distance Culling=true
    Inactive Response=::NewEnumerator2
    Life Cycle Mode=::NewEnumerator1
    Loop Behavior=::NewEnumerator2
    Loop Count:NiagaraInt32=&SelectIntFromArray(Direct Array Index=$Engine.QualityLevel, Int Selection Array=DataInterface() {propertyValues {IntData [10,10,15,20,20]}})
    Loop Delay=2.15
    Loop Duration=&Multiply_Float(
     A=&RandomRangeFloat(Minimum=0.0456, Maximum=0.0768)
     B=&SelectFloatFromArray(Direct Array Index=$Engine.QualityLevel, Float Selection Array=DataInterface() {propertyValues {FloatData [2,2,1.75,1.5,1.5]}})
    )
    MaxDistance=15000
    Module.bUseMaxDistance=true
    Module.bUseMinDistance=false
    Scalability Mode=::NewEnumerator1
    UseLoopDelay=true
   )
   &SpawnBurst_Instantaneous(Spawn Count=5, Use Spawn Probability=false, UseLoopCountLimit=false)
  ]
  particleSpawnScript [
   &PartitionParticles(
    Normalized Index In Partition Includes One=true
    Particles Per Partition=5
    Partition Count Mode:ENiagara_InfiniteLoopDuration=::NewEnumerator1
    Partition Mode:ENiagara_PartitionMode=::NewEnumerator1
    Write Partition Index to Ribbon Link Order:ENiagara_AttributeSamplingApplyOutput=::NewEnumerator0
    Write Partition to Ribbon ID=::NewEnumerator0
   )
   &InitializeParticle(
    Lifetime=0.8
    Ribbon Width=6
    Ribbon Width Mode=::NewEnumerator1
    Sprite Size Mode=::NewEnumerator3
    UsePositionOffset=false
    Write Material Random=true
   )
   &DynamicMaterialParameters(
    Index 0 Param 1=&RandomRangeFloat(
     Minimum=0
     Maximum=1
     Randomness Mode=::NewEnumerator1
     Random Seed=&Add_Integer(A=$Particles.PartitionParticles.Partition, B=3)
     Fixed Random Seed=true
    )
    Module.Index0Param0Write=true
    Module.Index0Param1Write=true
    Module.Index0Param2Write=true
    Module.Index0Param3Write=true
   )
  ]
  particleUpdateScript [
   &ParticleState()
   &ScaleColor(
    Scale RGB=&MakeVectorFromLinearColorRGB(Color=&ColorFromCurve(
      ColorCurve=DataInterface() {propertyValues {
        AlphaCurve {keys [<at 0 1 RCIM_Cubic>]}
        BlueCurve {keys [<at 0 0 RCIM_Cubic>,<at 0.7306 0.1562 RCIM_Cubic>]}
        GreenCurve {keys [<at 0 0.0174 RCIM_Cubic>,<at 0.7306 0.5609 RCIM_Cubic>]}
        RedCurve {keys [<at 0 0.3958 RCIM_Cubic>,<at 0.7306 1 RCIM_Cubic>]}
       }}
      CurveIndex=$Particles.RibbonLinkOrder
     ))
    ScaleA=true
    ScaleRGB=true
   )
   &CameraOffset(Camera Offset Amount=3)
   &ShapeLocation(
    Cylinder Distribution=::NewEnumerator2
    Cylinder Height Midpoint=0
    Cylinder Radius=&Lerp_Float(
     A=1
     B=25
     Alpha=&FloatFromCurve(
      FloatCurve=DataInterface() {propertyValues {Curve {keys [
          {arriveTangent=-1.8853, interpMode=RCIM_Cubic, leaveTangent=-1.8853, tangentMode=RCTM_User, time=0, value=1}
          {interpMode=RCIM_Cubic, tangentMode=RCTM_User, time=1, value=0}
         ]}}}
      CurveIndex=$Particles.RibbonLinkOrder
     )
    )
    MaskSpawnGroup=false
    Offset=&MakeVector(
     X=0
     Y=0
     Z=&Add_Float(
      A=&FloatFromCurve(
       FloatCurve=DataInterface() {propertyValues {Curve {keys [{tangentMode=RCTM_User, time=0, value=1}, {tangentMode=RCTM_User, time=1, value=0}]}}}
       Scale Curve=60
      )
      B=&Add_Float(
       A=25
       B=&FloatFromCurve(
        FloatCurve=DataInterface() {propertyValues {Curve {keys [<at 0 0>,<at 1 1>]}}}
        CurveIndex=$Particles.PartitionParticles.Normalized Index In Partition
        Scale Curve=-75
       )
      )
     )
    )
    Offset Mode=::NewEnumerator0
    Rotation Mode=::NewEnumerator2
    Shape Primitive=::NewEnumerator1
    U Position=&Multiply_Float(
     A=$Particles.PartitionParticles.Normalized Index In Partition
     B=&RandomRangeFloat(Minimum=1, Maximum=1.5, Randomness Mode=::NewEnumerator1, Random Seed=$Particles.PartitionParticles.Partition, Fixed Random Seed=true)
    )
    UseOwnerScale=true
    V Position=&FloatFromCurve(
     FloatCurve=DataInterface() {propertyValues {Curve {keys [
         {arriveTangent=0.2193, interpMode=RCIM_Cubic, leaveTangent=0.2193, tangentMode=RCTM_User, time=0, value=0}
         {arriveTangent=1.1185, tangentMode=RCTM_User, time=1, value=1}
        ]}}}
     CurveIndex=&RandomRangeFloat(Minimum=0.4, Maximum=0.5, Randomness Mode=::NewEnumerator1, Random Seed=$Particles.PartitionParticles.Partition, Fixed Random Seed=true)
    )
    Yaw / Pitch / Roll=&MakeVector(
     X=&Add_Float(
      A=&Multiply_Float(
       A=&RandomRangeFloat(Minimum=0.6, Maximum=0.6, Randomness Mode=::NewEnumerator1, Random Seed=$Particles.PartitionParticles.Partition, Fixed Random Seed=true)
       B=$Particles.Age
      )
      B=&Multiply_FloatByInt(Float=0.22, Integer=$Particles.PartitionParticles.Partition)
     )
     Y=0
     Z=0
    )
   )
   &ScaleRibbonWidth(Ribbon Width Scale=&FloatFromCurve(FloatCurve=DataInterface() {propertyValues {Curve {keys [
         {arriveTangent=3.1357, interpMode=RCIM_Cubic, leaveTangent=3.1357, tangentMode=RCTM_User, time=0.1, value=0}
         {interpMode=RCIM_Cubic, tangentMode=RCTM_User, time=0.5, value=1}
         {arriveTangent=-2.9338, interpMode=RCIM_Cubic, leaveTangent=-2.9338, tangentMode=RCTM_User, time=0.9, value=0}
        ]}}}))
  ]
  rendererClasses [&NiagaraRibbonRendererProperties]
  renderers [{rendererClass=&NiagaraRibbonRendererProperties}]
 }
 Spirals_Ribbon_Up {
  emitterUpdateScript [
   &EmitterState(
    Delay First Loop Only=true
    Enable Distance Culling=true
    Inactive Response=::NewEnumerator2
    Life Cycle Mode=::NewEnumerator1
    Loop Behavior=::NewEnumerator2
    Loop Count=10
    Loop Delay=1
    Loop Duration=&RandomRangeFloat(Minimum=0.111, Maximum=0.155)
    MaxDistance=15000
    Module.bUseMaxDistance=true
    Module.bUseMinDistance=false
    Scalability Mode=::NewEnumerator1
    UseLoopDelay=true
   )
   &SpawnBurst_Instantaneous(Spawn Count=5, Use Spawn Probability=false, UseLoopCountLimit=false)
  ]
  particleSpawnScript [
   &PartitionParticles(
    Normalized Index In Partition Includes One=true
    Particles Per Partition=5
    Partition Count Mode:ENiagara_InfiniteLoopDuration=::NewEnumerator1
    Partition Mode:ENiagara_PartitionMode=::NewEnumerator1
    Write Partition Index to Ribbon Link Order:ENiagara_AttributeSamplingApplyOutput=::NewEnumerator0
    Write Partition to Ribbon ID=::NewEnumerator0
   )
   &InitializeParticle(
    Lifetime=0.8
    Ribbon Width=6
    Ribbon Width Mode=::NewEnumerator1
    Sprite Size Mode=::NewEnumerator3
    UsePositionOffset=false
    Write Material Random=true
   )
   &DynamicMaterialParameters(
    Index 0 Param 1=&RandomRangeFloat(
     Minimum=0
     Maximum=1
     Randomness Mode=::NewEnumerator1
     Random Seed=&Add_Integer(A=$Particles.PartitionParticles.Partition, B=3)
     Fixed Random Seed=true
    )
    Module.Index0Param0Write=true
    Module.Index0Param1Write=true
    Module.Index0Param2Write=true
    Module.Index0Param3Write=true
   )
  ]
  particleUpdateScript [
   &ParticleState()
   &ScaleColor(
    Scale RGB=&MakeVectorFromLinearColorRGB(Color=&ColorFromCurve(
      ColorCurve=DataInterface() {propertyValues {
        AlphaCurve {keys [<at 0 1 RCIM_Cubic>]}
        BlueCurve {keys [<at 0 0 RCIM_Cubic>,<at 0.7306 0.1562 RCIM_Cubic>]}
        GreenCurve {keys [<at 0 0.0174 RCIM_Cubic>,<at 0.7306 0.5609 RCIM_Cubic>]}
        RedCurve {keys [<at 0 0.3958 RCIM_Cubic>,<at 0.7306 1 RCIM_Cubic>]}
       }}
      CurveIndex=$Particles.RibbonLinkOrder
     ))
    ScaleA=true
    ScaleRGB=true
   )
   &ShapeLocation(
    Cylinder Distribution=::NewEnumerator2
    Cylinder Height=75
    Cylinder Height Midpoint=0
    Cylinder Radius=&Lerp_Float(
     A=2
     B=25
     Alpha=&FloatFromCurve(
      FloatCurve=DataInterface() {propertyValues {Curve {keys [
          {arriveTangent=-1.8853, interpMode=RCIM_Cubic, leaveTangent=-1.8853, tangentMode=RCTM_User, time=0, value=1}
          {interpMode=RCIM_Cubic, tangentMode=RCTM_User, time=1, value=0}
         ]}}}
      CurveIndex=$Particles.RibbonLinkOrder
     )
    )
    MaskSpawnGroup=false
    Offset=&MakeVector(
     X=0
     Y=0
     Z=&Add_Float(
      A=&FloatFromCurve(
       FloatCurve=DataInterface() {propertyValues {Curve {keys [{tangentMode=RCTM_User, time=0, value=0}, {tangentMode=RCTM_User, time=1, value=1}]}}}
       Scale Curve=50
      )
      B=&Add_Float(
       A=0
       B=&FloatFromCurve(
        FloatCurve=DataInterface() {propertyValues {Curve {keys [<at 0 0>,<at 1 1>]}}}
        CurveIndex=$Particles.PartitionParticles.Normalized Index In Partition
        Scale Curve=-50
       )
      )
     )
    )
    Offset Mode=::NewEnumerator0
    Rotation Mode=::NewEnumerator2
    Shape Primitive=::NewEnumerator1
    U Position=&Multiply_Float(
     A=$Particles.PartitionParticles.Normalized Index In Partition
     B=&RandomRangeFloat(Minimum=0.5, Maximum=1, Randomness Mode=::NewEnumerator1, Random Seed=$Particles.PartitionParticles.Partition, Fixed Random Seed=true)
    )
    UseOwnerScale=true
    V Position=&FloatFromCurve(
     FloatCurve=DataInterface() {propertyValues {Curve {keys [
         {arriveTangent=0.2193, interpMode=RCIM_Cubic, leaveTangent=0.2193, tangentMode=RCTM_User, time=0, value=0}
         {arriveTangent=1.1185, tangentMode=RCTM_User, time=1, value=1}
        ]}}}
     CurveIndex=&RandomRangeFloat(Minimum=0.4, Maximum=0.5, Randomness Mode=::NewEnumerator1, Random Seed=$Particles.PartitionParticles.Partition, Fixed Random Seed=true)
    )
    Yaw / Pitch / Roll=&MakeVector(
     X=&Add_Float(
      A=&Multiply_Float(
       A=&RandomRangeFloat(Minimum=-0.8, Maximum=-0.8, Randomness Mode=::NewEnumerator1, Random Seed=$Particles.PartitionParticles.Partition, Fixed Random Seed=true)
       B=$Particles.Age
      )
      B=&Multiply_FloatByInt(Float=0.22, Integer=$Particles.PartitionParticles.Partition)
     )
     Y=0
     Z=0
    )
   )
   &ScaleRibbonWidth(Ribbon Width Scale=&FloatFromCurve(FloatCurve=DataInterface() {propertyValues {Curve {keys [
         {arriveTangent=2.4561, interpMode=RCIM_Cubic, leaveTangent=2.4561, tangentMode=RCTM_User, time=0.1, value=0}
         {interpMode=RCIM_Cubic, tangentMode=RCTM_User, time=0.5, value=1}
         {arriveTangent=-2.4926, interpMode=RCIM_Cubic, leaveTangent=-2.4926, tangentMode=RCTM_User, time=0.9, value=0}
        ]}}}))
  ]
  rendererClasses [&NiagaraRibbonRendererProperties]
  renderers [{rendererClass=&NiagaraRibbonRendererProperties}]
 }
 Tornado {
  emitterUpdateScript [
   &EmitterState(Inactive Response=::NewEnumerator1, Life Cycle Mode=::NewEnumerator1, Loop Behavior=::NewEnumerator1, Loop Delay=1, UseLoopDelay=true)
   &BeamEmitterSetup(Beam End:Vector3f=<0 0 125>)
   &SpawnBurst_Instantaneous(Spawn Count=16, Use Spawn Probability=false, UseLoopCountLimit=false)
  ]
  particleSpawnScript [
   &InitializeParticle(Color Mode=::NewEnumerator0, Lifetime=4.25, Position Mode=::NewEnumerator0, Ribbon Width Mode=::NewEnumerator1, Write Material Random=true)
   &SpawnBeam()
   &BeamWidth(Beam Width:NiagaraFloat=&FloatFromCurve(
     FloatCurve=DataInterface() {propertyValues {Curve {keys [
         {arriveTangent=0.0019, interpMode=RCIM_Cubic, time=0, value=1}
         {arriveTangent=11.6802, interpMode=RCIM_Cubic, leaveTangent=11.6802, tangentMode=RCTM_User, time=1, value=5}
        ]}}}
     CurveIndex=$Particles.RibbonLinkOrder
     Scale Curve=22
    ))
  ]
  particleUpdateScript [
   &SkeletalMeshLocation(Bone Sampling Mode=::NewEnumerator4, MaskSpawnGroup=false, Skeletal Mesh=$System.Skeletal Mesh)
   &ParticleState()
   /SetVariables(Particles.InnerWidth:NiagaraFloat=&FloatFromCurve(
     FloatCurve=DataInterface() {propertyValues {Curve {keys [
         <at 0 0.2 RCIM_Cubic>
         {arriveTangent=0.9171, interpMode=RCIM_Cubic, leaveTangent=0.9171, tangentMode=RCTM_User, time=0.5, value=0.525}
         {arriveTangent=1.6745, interpMode=RCIM_Cubic, leaveTangent=1.6745, tangentMode=RCTM_User, time=1, value=2}
        ]}}}
     CurveIndex=$Particles.RibbonLinkOrder
     Scale Curve=&Multiply_Float(
      A=45
      B=&Multiply_Float(
       A=&FloatFromCurve(FloatCurve=DataInterface() {propertyValues {Curve {keys [
            {interpMode=RCIM_Cubic, tangentMode=RCTM_User, time=0, value=0}
            {arriveTangent=1.1111, leaveTangent=1.1111, time=0.25, value=1}
            {interpMode=RCIM_Cubic, tangentMode=RCTM_User, time=0.9, value=1}
            {tangentMode=RCTM_User, time=1, value=0}
           ]}}})
       B=&Sine(Normalized Angle=&Add_Float(A=$Emitter.Age, B=$Particles.RibbonLinkOrder), Period=0.5, Scale=0.2, Bias=1)
      )
     )
    )) {isSetParameters=true}
   /SetVariables(Particles.OuterLayer2Width:NiagaraFloat=&FloatFromCurve(
     FloatCurve=DataInterface() {propertyValues {Curve {keys [
         <at 0 0.2 RCIM_Cubic>
         {arriveTangent=0.9171, interpMode=RCIM_Cubic, leaveTangent=0.9171, tangentMode=RCTM_User, time=0.5, value=0.525}
         {arriveTangent=1.6745, interpMode=RCIM_Cubic, leaveTangent=1.6745, tangentMode=RCTM_User, time=1, value=2}
        ]}}}
     CurveIndex=$Particles.RibbonLinkOrder
     Scale Curve=&Multiply_Float(
      A=43.0928
      B=&Multiply_Float(
       A=&FloatFromCurve(FloatCurve=DataInterface() {propertyValues {Curve {keys [
            {interpMode=RCIM_Cubic, tangentMode=RCTM_User, time=0, value=0}
            {arriveTangent=1.1111, leaveTangent=1.1111, time=0.25, value=1}
            {interpMode=RCIM_Cubic, tangentMode=RCTM_User, time=0.9, value=1}
            {tangentMode=RCTM_User, time=1, value=0}
           ]}}})
       B=&Sine(Normalized Angle=&Add_Float(A=$Emitter.Age, B=$Particles.RibbonLinkOrder), Period=0.5, Scale=0.2, Bias=1)
      )
     )
    )) {isSetParameters=true}
   /Niagara/Modules/Beams/ScaleBeamWidth(Beam Width Scale:NiagaraFloat=&Multiply_Float(
     A=&FloatFromCurve(FloatCurve=DataInterface() {propertyValues {Curve {keys [
          {interpMode=RCIM_Cubic, tangentMode=RCTM_User, time=0, value=0}
          {arriveTangent=1.1111, leaveTangent=1.1111, time=0.2, value=1}
          {interpMode=RCIM_Cubic, tangentMode=RCTM_User, time=0.9, value=1}
          {tangentMode=RCTM_User, time=1, value=0}
         ]}}})
     B=&Sine(Normalized Angle=&Add_Float(A=$Emitter.Age, B=$Particles.RibbonLinkOrder), Period=0.5, Scale=0.15, Bias=1)
    ))
   &UpdateBeam()
   &SolveForcesAndVelocity(Clamp Velocity=false, Limit Acceleration=false)
   &DynamicMaterialParameters(
    Index 0 Param 1=&FloatFromCurve(FloatCurve=DataInterface() {propertyValues {Curve {keys [
         {interpMode=RCIM_Cubic, tangentMode=RCTM_User, time=0, value=0}
         <at 0.2 1>
         {arriveTangent=-1.25, interpMode=RCIM_Cubic, leaveTangent=-1.25, time=0.75, value=1}
         {arriveTangent=-5.957, interpMode=RCIM_Cubic, leaveTangent=-5.957, tangentMode=RCTM_User, time=1, value=0}
        ]}}})
    Index 0 Param 2=&FloatFromCurve(FloatCurve=DataInterface() {propertyValues {Curve {keys [
         {arriveTangent=-7.1115, interpMode=RCIM_Cubic, leaveTangent=-7.1115, tangentMode=RCTM_User, time=0, value=0.5}
         {arriveTangent=0.3823, interpMode=RCIM_Cubic, leaveTangent=0.3823, tangentMode=RCTM_User, time=0.35, value=-1}
         {arriveTangent=5.2808, interpMode=RCIM_Cubic, leaveTangent=5.2808, tangentMode=RCTM_User, time=1, value=2}
        ]}}})
    Index 0 Param 3=&Lerp_Float(
     A=0.25
     B=1
     Alpha=&FloatFromCurve(FloatCurve=DataInterface() {propertyValues {Curve {keys [
          {arriveTangent=3.2848, interpMode=RCIM_Cubic, leaveTangent=3.2848, tangentMode=RCTM_User, time=0, value=0}
          {arriveTangent=1.4286, leaveTangent=1.4286, time=0.3, value=1}
          {interpMode=RCIM_Cubic, tangentMode=RCTM_User, time=0.7, value=1}
          {tangentMode=RCTM_User, time=1, value=0}
         ]}}})
    )
    Index 0 Param 4=&Lerp_Float(
     A=1
     B=2
     Alpha=&FloatFromCurve(FloatCurve=DataInterface() {propertyValues {Curve {keys [
          {interpMode=RCIM_Cubic, tangentMode=RCTM_User, time=0, value=0}
          {interpMode=RCIM_Cubic, tangentMode=RCTM_User, time=0.5, value=1}
          {tangentMode=RCTM_User, time=1, value=0}
         ]}}})
    )
    Module.Index0Param0Write=true
    Module.Index0Param1Write=true
    Module.Index0Param2Write=true
    Module.Index0Param3Write=true
   )
   &ScaleColor(
    Scale Alpha=&FloatFromCurve(FloatCurve=DataInterface() {propertyValues {Curve {keys [<at 0.75 1>,<at 0.9 0>]}}})
    ScaleA=true
    ScaleRGB=true
   )
   /SetVariables(Particles.Layer 2 Dynamic:Vector4f=&MakeVector4(
     X=&FloatFromCurve(FloatCurve=DataInterface() {propertyValues {Curve {keys [
          {interpMode=RCIM_Cubic, tangentMode=RCTM_User, time=0, value=0}
          <at 0.2 1>
          {arriveTangent=-1.25, interpMode=RCIM_Cubic, leaveTangent=-1.25, time=0.75, value=1}
          {arriveTangent=-5.957, interpMode=RCIM_Cubic, leaveTangent=-5.957, tangentMode=RCTM_User, time=1, value=0}
         ]}}})
     Y=&FloatFromCurve(
      FloatCurve=DataInterface() {propertyValues {Curve {keys [
          {arriveTangent=-7.1115, interpMode=RCIM_Cubic, leaveTangent=-7.1115, tangentMode=RCTM_User, time=0, value=0.5}
          {arriveTangent=0.3823, interpMode=RCIM_Cubic, leaveTangent=0.3823, tangentMode=RCTM_User, time=0.4, value=-1}
          {arriveTangent=5.2808, interpMode=RCIM_Cubic, leaveTangent=5.2808, tangentMode=RCTM_User, time=1, value=2}
         ]}}}
      Scale Curve=1.25
     )
     Z=&Lerp_Float(
      A=0.5
      B=1.25
      Alpha=&FloatFromCurve(FloatCurve=DataInterface() {propertyValues {Curve {keys [
           {arriveTangent=3.2848, interpMode=RCIM_Cubic, leaveTangent=3.2848, tangentMode=RCTM_User, time=0, value=0}
           {arriveTangent=1.4286, leaveTangent=1.4286, time=0.3, value=1}
           {interpMode=RCIM_Cubic, tangentMode=RCTM_User, time=0.7, value=1}
           {tangentMode=RCTM_User, time=1, value=0}
          ]}}})
     )
     W=&Lerp_Float(
      A=1
      B=3
      Alpha=&FloatFromCurve(FloatCurve=DataInterface() {propertyValues {Curve {keys [
           {interpMode=RCIM_Cubic, tangentMode=RCTM_User, time=0, value=0}
           {interpMode=RCIM_Cubic, tangentMode=RCTM_User, time=0.5, value=1}
           {tangentMode=RCTM_User, time=1, value=0}
          ]}}})
     )
    )) {isSetParameters=true}
  ]
  rendererClasses [&NiagaraRibbonRendererProperties]
  renderers [
   {rendererClass=&NiagaraRibbonRendererProperties}
   {rendererClass=&NiagaraRibbonRendererProperties, rendererIndex=1}
   {rendererClass=&NiagaraRibbonRendererProperties, rendererIndex=2}
  ]
 }
}
summary {
 systemName=NS_Emote_Gatehound_Box
 emitters [
  {emitterName=GlowFlare001, rendererClasses [&NiagaraSpriteRendererProperties]}
  {emitterName=Flare, rendererClasses [&NiagaraSpriteRendererProperties, &NiagaraLightRendererProperties]}
  {emitterName=Shine, rendererClasses [&NiagaraSpriteRendererProperties]}
  {emitterName=FlareSkull, rendererClasses [&NiagaraSpriteRendererProperties, &NiagaraLightRendererProperties]}
  {emitterName=Glowbeam, rendererClasses [&NiagaraRibbonRendererProperties]}
  {emitterName=SparkDown, simTarget=GPUComputeSim, rendererClasses [&NiagaraSpriteRendererProperties]}
  {emitterName=LightShaftsBurst, rendererClasses [&NiagaraMeshRendererProperties]}
  {emitterName=SmokeBlackDetail, rendererClasses [&NiagaraSpriteRendererProperties]}
  {emitterName=BoxFlare, rendererClasses [&NiagaraSpriteRendererProperties]}
  {emitterName=SparksOpen, simTarget=GPUComputeSim, rendererClasses [&NiagaraSpriteRendererProperties]}
  {emitterName=Tornado, rendererClasses [&NiagaraRibbonRendererProperties]}
  {emitterName=SparkUp, simTarget=GPUComputeSim, rendererClasses [&NiagaraSpriteRendererProperties]}
  {emitterName=Spirals_Ribbon_Up, rendererClasses [&NiagaraRibbonRendererProperties]}
  {emitterName=Spirals_Ribbon_Down, rendererClasses [&NiagaraRibbonRendererProperties]}
  {emitterName=Glowbeam_Soft, rendererClasses [&NiagaraRibbonRendererProperties]}
  {emitterName=SmokeBlackDetail001, rendererClasses [&NiagaraSpriteRendererProperties]}
 ]
 userVariables [
  User.BeamErode {type=NiagaraFloat, defaultValue=1, description=""}
  User.BeamGrow {type=NiagaraFloat, defaultValue=1, description=""}
  User.bIsFrontEnd {type=NiagaraBool}
  User.bIsFrontEndPreview {type=NiagaraBool}
  User.GlowColor {type=LinearColor, description="", defaultValue {struct=LinearColor, value=<rgba 1 0.4549 0 1>}}
 ]
}
