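# Niagara effect for the Gatehound emote: three skulls with glow sprites, trailed by fiery ribbons, smoke and GPU sparks.
# All four emitters split particles into three partitions, each sampling its own bone or socket index on a skeletal mesh.
# Trail, smoke and spark particles share vortex and point-attraction forces, ramped by system loop age, that pull them toward a system-sampled mesh position.
# Skulls brighten and shrink late in life as the trail color shifts to orange.
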NiagaraSystem /BRCosmetics/Animation/Game/MainPlayer/Emotes/GateHound/FX/Systems/NS_Emote_Gatehound_Skulltrail
dependencies {
 usedDataInterfaces [
  &NiagaraDataInterfaceArrayInt32
  &NiagaraDataInterfaceCurve
  &NiagaraDataInterfaceColorCurve
  &NiagaraDataInterfaceVector2DCurve
  &NiagaraDataInterfaceSpriteRendererInfo
  &NiagaraDataInterfaceSkeletalMesh
 ]
 usedDynamicInputs [&Vector2DFromCurve, &RandomRangeVector2D]
 usedModules [
  &self:SystemSpawnScript.SystemScriptSource.SystemScriptGraph.NiagaraNodeAssignment_0.SetVariables_#1
  &self:SystemSpawnScript.SystemScriptSource.SystemScriptGraph.NiagaraNodeAssignment_1.SetVariables_#2
  &SystemState
  &SampleSkeletalMesh
 ]
}
emitters {
 SkullTrail {
  emitterUpdateScript [
   &EmitterState()
   &SpawnRate(SpawnRate=120, Use Spawn Probability=false)
  ]
  particleSpawnScript [
   &PartitionParticles(Number Of Partitions=3)
   &InitializeParticle(
    Color Mode=::NewEnumerator0
    Lifetime=0.225
    Ribbon Width=15
    Ribbon Width Mode=::NewEnumerator1
    UsePositionOffset=false
    Write Material Random=true
   )
   &SkeletalMeshLocation(
    Bone / Socket Index=&SelectIntFromArray(Direct Array Index=$Particles.PartitionParticles.Partition, Int Selection Array=DataInterface() {propertyValues {IntData [0,1,2]}})
    Bone Sampling Mode=::NewEnumerator4
    MaskSpawnGroup=false
    Sampled Position Offset=<0 0 5>
    Sampled Position Offset Coordinate Space:ENiagaraSkelMeshTransforms=::NewEnumerator4
    Skeletal Mesh=$System.Skeletal Mesh
   )
   &DynamicMaterialParameters(
    Index 0 Param 1=10
    Index 0 Param 2=10
    Index 0 Param 3=1
    Index 0 Param 4=1
    Module.Index0Param0Write=true
    Module.Index0Param1Write=true
    Module.Index0Param2Write=true
    Module.Index0Param3Write=true
   )
  ]
  particleUpdateScript [
   &ParticleState()
   &InheritSourceMovement(
    Applied Position Scale=&Multiply_VectorByFloat(
     Vector=<1 1 1>
     Float=&FloatFromCurve(FloatCurve=DataInterface() {propertyValues {Curve {keys [{interpMode=RCIM_Cubic, tangentMode=RCTM_User, time=0, value=1}, {interpMode=RCIM_Cubic, tangentMode=RCTM_User, time=0.25, value=0}]}}})
    )
    ApplyForce=false
    ApplyPosition=true
    ApplyPositionFalloff=false
   )
   &VortexForce(
    Origin Pull Amount:NiagaraFloat=&Lerp_Float(
     A=0
     B=2000
     Alpha=&FloatFromCurve(
      FloatCurve=DataInterface() {propertyValues {Curve {keys [{interpMode=RCIM_Cubic, tangentMode=RCTM_User, time=0.4, value=0}, {interpMode=RCIM_Cubic, tangentMode=RCTM_User, time=0.9, value=1}]}}}
      CurveIndex=$System.NormalizedLoopAge
     )
    )
    UseOriginOffset=true
    Vortex Force Amount=&Lerp_Float(
     A=0
     B=1000
     Alpha=&FloatFromCurve(
      FloatCurve=DataInterface() {propertyValues {Curve {keys [<at 0.4 0 RCIM_Cubic>,<at 0.7 1 RCIM_Cubic>]}}}
      CurveIndex=$System.NormalizedLoopAge
     )
    )
    Vortex Origin:NiagaraPosition=$System.SampleSkeletalMesh.SampledPosition
    Vortex Origin Offset:Vector3f=<0 0 25>
   )
   &PointAttractionForce(
    Attraction Radius=500
    Attraction Strength=&Multiply_Float(
     A=&FloatFromCurve(
      FloatCurve=DataInterface() {propertyValues {Curve {keys [<at 0.5 0 RCIM_Cubic>,<at 0.8 1 RCIM_Cubic>]}}}
      CurveIndex=$System.NormalizedLoopAge
     )
     B=&Lerp_Float(
      A=0
      B=700
      Alpha=&FloatFromCurve(FloatCurve=DataInterface() {propertyValues {Curve {keys [{tangentMode=RCTM_User, time=0, value=0}, {tangentMode=RCTM_User, time=1, value=1}]}}})
     )
    )
    Attractor Position:NiagaraPosition=$System.SampleSkeletalMesh.SampledPosition
    Attractor Position Offset:Vector3f=<0 0 35>
    Kill Within Radius=false
    Module.Use Falloff=true
    UseAttractorOffset=true
   )
   &SolveForcesAndVelocity(Clamp Velocity=false, Limit Acceleration=false)
   &ScaleRibbonWidth(Ribbon Width Scale=&FloatFromCurve(FloatCurve=DataInterface() {propertyValues {Curve {keys [<at 0.4 1 RCIM_Cubic>,<at 0.9 0 RCIM_Cubic>]}}}))
   &ScaleColor(
    Scale Alpha=&FloatFromCurve(
     FloatCurve=DataInterface() {propertyValues {Curve {keys [
         {interpMode=RCIM_Cubic, tangentMode=RCTM_User, time=0, value=0}
         {arriveTangent=0.6078, leaveTangent=0.6078, time=0.2, value=1}
         {interpMode=RCIM_Cubic, tangentMode=RCTM_User, time=0.7, value=1}
        ]}}}
     CurveIndex=$System.NormalizedLoopAge
    )
    Scale RGB=&MakeVectorFromLinearColorRGB(Color=&LerpLinearColors(
      StartColor=&ColorFromCurve(ColorCurve=DataInterface() {propertyValues {
         AlphaCurve {keys [<at 0 1 RCIM_Cubic>,<at 1 0 RCIM_Cubic>]}
         BlueCurve {keys [<at 0 0 RCIM_Cubic>,<at 0.4903 0.0431>,<at 1 0 RCIM_Cubic>]}
         GreenCurve {keys [<at 0 1.6493 RCIM_Cubic>,<at 0.4903 0.1961>,<at 1 0.0714 RCIM_Cubic>]}
         RedCurve {keys [<at 0 4 RCIM_Cubic>,<at 0.4903 0.5451>,<at 1 0.3906 RCIM_Cubic>]}
        }})
      EndColor=<rgba 2 0.6019 0.0364 1>
      LerpFactor=&FloatFromCurve(
       FloatCurve=DataInterface() {propertyValues {Curve {keys [<at 0.75 0 RCIM_Cubic>,<at 1 1 RCIM_Cubic>]}}}
       CurveIndex=$System.NormalizedLoopAge
      )
     ))
    ScaleA=true
    ScaleRGB=true
   )
   &FN_RibbonMaterial(
    Color=&LerpLinearColors(
     StartColor=&ColorFromCurve(ColorCurve=DataInterface() {propertyValues {
        AlphaCurve {keys [<at 0 1 RCIM_Cubic>,<at 1 0 RCIM_Cubic>]}
        BlueCurve {keys [<at 0.0092 0.0521>, {arriveTangent=-0.0526, interpMode=RCIM_Cubic, leaveTangent=-0.0526, time=0.231, value=0}, <at 1 0 RCIM_Cubic>]}
        GreenCurve {keys [<at 0.0092 0.4094>, {arriveTangent=-0.3748, interpMode=RCIM_Cubic, leaveTangent=-0.3748, time=0.231, value=0.1345}, <at 1 0.038 RCIM_Cubic>]}
        RedCurve {keys [<at 0.0092 1>, {arriveTangent=-0.9041, interpMode=RCIM_Cubic, leaveTangent=-0.9041, time=0.231, value=0.5208}, <at 1 0.1042 RCIM_Cubic>]}
       }})
     EndColor=<rgba 2 0.6019 0.0364 1>
     LerpFactor=&FloatFromCurve(
      FloatCurve=DataInterface() {propertyValues {Curve {keys [<at 0.75 0 RCIM_Cubic>,<at 1 1 RCIM_Cubic>]}}}
      CurveIndex=$System.NormalizedLoopAge
     )
    )
    Opacity Erosion:NiagaraFloat=&FloatFromCurve(FloatCurve=DataInterface() {propertyValues {Curve {keys [<at 0 1>,<at 1 0>]}}})
    Scale Alpha:NiagaraFloat=&FloatFromCurve(
     FloatCurve=DataInterface() {propertyValues {Curve {keys [
         <at 0 0 RCIM_Cubic>
         {arriveTangent=1.4286, leaveTangent=1.4286, time=0.2, value=1}
         {interpMode=RCIM_Cubic, tangentMode=RCTM_User, time=0.7, value=1}
         {tangentMode=RCTM_User, time=1, value=0}
        ]}}}
     CurveIndex=$System.NormalizedLoopAge
    )
    Write to Module Namespace=true
   )
   /SetVariables(Particles.BlackRibbonDynamic:Vector4f=&MakeVector4(X=1, Y=1, Z=2, W=1)) {isSetParameters=true}
  ]
  rendererClasses [&NiagaraRibbonRendererProperties]
  renderers [
   {rendererClass=&NiagaraRibbonRendererProperties}
   {rendererClass=&NiagaraRibbonRendererProperties, rendererIndex=1}
  ]
 }
 Skulls_Main {
  emitterUpdateScript [
   &EmitterState()
   &SpawnBurst_Instantaneous(Spawn Count=3, Use Spawn Probability=false, UseLoopCountLimit=false)
  ]
  particleSpawnScript [
   &PartitionParticles(Number Of Partitions=3)
   &InitializeParticle(
    Color=<rgba 0.7031 0.3397 0.2124 1>
    Lifetime=5.5
    Mesh Scale Mode=::NewEnumerator3
    Mesh Uniform Scale=0.6
    Sprite Size Mode=::NewEnumerator3
    Uniform Sprite Size=30
    UsePositionOffset=true
    Write Material Random=true
   )
   &InitialMeshOrientation(
    Mesh Axis To Orient=<0 0 1>
    Mesh Orientation Mode=::NewEnumerator2
    Orientation Vector=<0 -1 0>
    Rotation=<0 0 0.25>
    Rotation Coordinate Space=::NewEnumerator0
    Use Rotation Vector=true
   )
   &DynamicMaterialParameters(
    Index 0 Param 1=1
    Index 0 Param 2=1
    Module.Index0Param0Write=true
    Module.Index0Param1Write=true
    Module.Index0Param2Write=true
    Module.Index0Param3Write=true
   )
   /SetVariables(Particles.SpriteDynamic:Vector4f=<3 1 1 1>) {isSetParameters=true}
  ]
  particleUpdateScript [
   &SkeletalMeshLocation(
    Bone / Socket Index=&SelectIntFromArray(Direct Array Index=$Particles.PartitionParticles.Partition, Int Selection Array=DataInterface() {propertyValues {IntData [0,1,2]}})
    Bone Sampling Mode=::NewEnumerator4
    MaskSpawnGroup=false
    Sampled Position Offset=<0 0 5>
    Skeletal Mesh=$System.Skeletal Mesh
    Velocity Sampling=::NewEnumerator0
    VelocitySpeedLimit=false
   )
   &ScaleColor(
    Scale Alpha=&FloatFromCurve(FloatCurve=DataInterface() {propertyValues {Curve {keys [
         {interpMode=RCIM_Cubic, tangentMode=RCTM_User, time=0, value=0}
         {arriveTangent=1.25, time=0.1, value=1}
         {interpMode=RCIM_Cubic, tangentMode=RCTM_User, time=0.85, value=1}
         <at 0.95 0 RCIM_Cubic>
        ]}}})
    Scale RGB=&Lerp_Vector(
     A=&MakeVectorFromLinearColorRGB(Color=$Particles.Initial.Color)
     B=&MakeVectorFromLinearColorRGB(Color=<rgba 15 10.1 0 1>)
     Alpha=&FloatFromCurve(FloatCurve=DataInterface() {propertyValues {Curve {keys [<at 0.85 0 RCIM_Cubic>,<at 1.2 1 RCIM_Cubic>]}}})
    )
    ScaleA=true
    ScaleRGB=true
   )
   &ParticleState()
   &ScaleMeshSize(Scale Factor=&Multiply_VectorByFloat(
     Vector=<1 1 1>
     Float=&FloatFromCurve(FloatCurve=DataInterface() {propertyValues {Curve {keys [<at 0.9 1 RCIM_Cubic>,<at 1 0 RCIM_Cubic>]}}})
    ))
   /SetVariables(
    Particles.GlowCameraOffset=-10
    Particles.GlowColor=&MakeLinearColorFromVectorAndFloat(
     "Vector (RGB)"=&MakeVectorFromLinearColorRGB(Color=<rgba 0.6875 0.1751 0.0859 1>)
     "Float (Alpha)"=&FloatFromCurve(
      FloatCurve=DataInterface() {propertyValues {Curve {keys [
          {interpMode=RCIM_Cubic, tangentMode=RCTM_User, time=0.05, value=0}
          {arriveTangent=1.25, time=0.25, value=1}
          {interpMode=RCIM_Cubic, tangentMode=RCTM_User, time=0.8, value=1}
         ]}}}
      Scale Curve=0.5
     )
    )
   ) {isSetParameters=true}
  ]
  rendererClasses [&NiagaraMeshRendererProperties, &NiagaraSpriteRendererProperties]
  renderers [
   {rendererClass=&NiagaraMeshRendererProperties, rendererIndex=0}
   {rendererClass=&NiagaraSpriteRendererProperties, rendererIndex=1}
  ]
 }
 SmokeDetail {
  emitterUpdateScript [
   &EmitterState(Inactive Response=::NewEnumerator2, Life Cycle Mode=::NewEnumerator1, Loop Behavior=::NewEnumerator1, Loop Duration=3, UseLoopDelay=true)
   &SpawnRate(SpawnRate=18, Use Spawn Probability=false)
  ]
  particleSpawnScript [
   &PartitionParticles(Number Of Partitions=3)
   &InitializeParticle(
    Lifetime Max=0.6
    Lifetime Min=0.4
    Lifetime Mode=::NewEnumerator1
    Sprite Rotation Mode=::NewEnumerator2
    Sprite Size Max=<25 25>
    Sprite Size Min=<20 20>
    Sprite Size Mode=::NewEnumerator2
    UsePositionOffset=false
    Write Material Random=true
   )
   &SkeletalMeshLocation(
    Bone / Socket Index=&SelectIntFromArray(Direct Array Index=$Particles.PartitionParticles.Partition, Int Selection Array=DataInterface() {propertyValues {IntData [0,1,2]}})
    Bone Sampling Mode=::NewEnumerator4
    MaskSpawnGroup=false
    Skeletal Mesh=$System.Skeletal Mesh
   )
  ]
  particleUpdateScript [
   &ParticleState()
   &VortexForce(
    Origin Pull Amount:NiagaraFloat=&Lerp_Float(
     A=0
     B=500
     Alpha=&FloatFromCurve(
      FloatCurve=DataInterface() {propertyValues {Curve {keys [{interpMode=RCIM_Cubic, tangentMode=RCTM_User, time=0.4, value=0}, {interpMode=RCIM_Cubic, tangentMode=RCTM_User, time=0.9, value=1}]}}}
      CurveIndex=$System.NormalizedLoopAge
     )
    )
    UseOriginOffset=true
    Vortex Force Amount=&Lerp_Float(
     A=0
     B=250
     Alpha=&FloatFromCurve(
      FloatCurve=DataInterface() {propertyValues {Curve {keys [<at 0.4 0 RCIM_Cubic>,<at 0.7 1 RCIM_Cubic>]}}}
      CurveIndex=$System.NormalizedLoopAge
     )
    )
    Vortex Origin:NiagaraPosition=$System.SampleSkeletalMesh.SampledPosition
    Vortex Origin Offset:Vector3f=<0 0 25>
   )
   &PointAttractionForce(
    Attraction Radius=600
    Attraction Strength=&Multiply_Float(
     A=&FloatFromCurve(
      FloatCurve=DataInterface() {propertyValues {Curve {keys [<at 0.5 0 RCIM_Cubic>,<at 0.8 1 RCIM_Cubic>]}}}
      CurveIndex=$System.NormalizedLoopAge
     )
     B=&Lerp_Float(
      A=0
      B=500
      Alpha=&FloatFromCurve(FloatCurve=DataInterface() {propertyValues {Curve {keys [{tangentMode=RCTM_User, time=0, value=0}, {tangentMode=RCTM_User, time=1, value=1}]}}})
     )
    )
    Attractor Position:NiagaraPosition=$System.SampleSkeletalMesh.SampledPosition
    Attractor Position Offset:Vector3f=<0 0 35>
    Kill Within Radius=false
    Module.Use Falloff=true
    UseAttractorOffset=true
   )
   &SpriteRotationRate(Rotation Rate=&RandomRangeFloat(Minimum=-50, Maximum=50))
   &ScaleSpriteSize(
    Scale Sprite Size Mode=::NewEnumerator0
    Uniform Scale Factor=&FloatFromCurve(FloatCurve=DataInterface() {propertyValues {Curve {keys [<at 0 0.8>,<at 1 1.2>]}}}, Scale Curve=1.3)
   )
   &SubUVAnimation(
    End Frame Range Override=30
    "Sprite Renderer "=DataInterface() {propertyValues {SpriteRenderer=&self:SmokeDetail_3.NiagaraSpriteRendererProperties_0}}
    Start Frame Range Override=&RandomRangeInt(Minimum=0, Maximum=5)
    UseEndFrame=true
    UseStartFrame=true
   )
   &ScaleColor(
    Scale Alpha=&FloatFromCurve(
     FloatCurve=DataInterface() {propertyValues {Curve {keys [
         {arriveTangent=9.0725, interpMode=RCIM_Cubic, leaveTangent=9.0725, tangentMode=RCTM_User, time=0, value=0}
         {arriveTangent=-2.5249, interpMode=RCIM_Cubic, leaveTangent=-2.5249, tangentMode=RCTM_User, time=0.1408, value=0.9847}
         {arriveTangent=-0.0405, interpMode=RCIM_Cubic, leaveTangent=-0.0405, tangentMode=RCTM_User, time=1, value=0}
        ]}}}
     Scale Curve=&FloatFromCurve(
      FloatCurve=DataInterface() {propertyValues {Curve {keys [{tangentMode=RCTM_User, time=0, value=0}, {tangentMode=RCTM_User, time=0.2, value=1}]}}}
      CurveIndex=$System.NormalizedLoopAge
     )
    )
    Scale RGB=&Multiply_VectorByFloat(
     Vector=&MakeVectorFromLinearColorRGB(Color=&ColorFromCurve(ColorCurve=DataInterface() {propertyValues {
         AlphaCurve {keys [<at 0 1 RCIM_Cubic>]}
         BlueCurve {keys [<at 0 0.0603 RCIM_Cubic>,<at 0.1853 0.0431 RCIM_Cubic>]}
         GreenCurve {keys [<at 0 0.6238 RCIM_Cubic>,<at 0.1853 0.1961 RCIM_Cubic>]}
         RedCurve {keys [<at 0 1 RCIM_Cubic>,<at 0.1853 0.5451 RCIM_Cubic>]}
        }}))
     Float=&FloatFromCurve(FloatCurve=DataInterface() {propertyValues {Curve {keys [{interpMode=RCIM_Cubic, tangentMode=RCTM_User, time=0, value=1}, {interpMode=RCIM_Cubic, tangentMode=RCTM_User, time=0.5, value=0.5}]}}})
    )
    ScaleA=true
    ScaleRGB=true
   )
   &SolveForcesAndVelocity(Clamp Velocity=false, Limit Acceleration=false)
  ]
  rendererClasses [&NiagaraSpriteRendererProperties]
  renderers [{rendererClass=&NiagaraSpriteRendererProperties}]
 }
 Sparks {
  simTarget=GPUComputeSim
  emitterUpdateScript [
   &EmitterState(
    Inactive Response=::NewEnumerator2
    Life Cycle Mode=::NewEnumerator1
    Loop Behavior=::NewEnumerator1
    Loop Delay=1
    Loop Duration=4.5
    UseLoopDelay=true
   )
   &SpawnRate(
    Spawn Probability=&FloatFromCurve(
     FloatCurve=DataInterface() {propertyValues {Curve {keys [<at 0 0>,<at 0.1 1>,<at 0.8 1>,<at 0.9 0>]}}}
     CurveIndex=$Emitter.NormalizedLoopAge
    )
    SpawnRate=50
    Use Spawn Probability=true
   )
  ]
  particleSpawnScript [
   &PartitionParticles(Number Of Partitions=3)
   &InitializeParticle(
    Color Mode=::NewEnumerator0
    Lifetime Max=0.75
    Lifetime Min=0.4
    Lifetime Mode=::NewEnumerator1
    Sprite Size Max=<2 2>
    Sprite Size Min=<1.5 1.5>
    Sprite Size Mode=::NewEnumerator2
    UsePositionOffset=false
    Write Material Random=true
   )
   &SkeletalMeshLocation(
    Bone / Socket Index=&SelectIntFromArray(Direct Array Index=$Particles.PartitionParticles.Partition, Int Selection Array=DataInterface() {propertyValues {IntData [0,1,2]}})
    Bone Sampling Mode=::NewEnumerator4
    MaskSpawnGroup=false
    Skeletal Mesh=$System.Skeletal Mesh
   )
   &ShapeLocation(
    MaskSpawnGroup=false
    Override Seed=false
    Shape Origin=$StackContext.SkeletalMeshLocation.SampledPosition
    Sphere Radius=5
    UseOwnerScale=true
   )
  ]
  particleUpdateScript [
   &ParticleState()
   &VortexForce(
    Origin Pull Amount:NiagaraFloat=&Lerp_Float(
     A=0
     B=500
     Alpha=&FloatFromCurve(
      FloatCurve=DataInterface() {propertyValues {Curve {keys [{interpMode=RCIM_Cubic, tangentMode=RCTM_User, time=0.4, value=0}, {interpMode=RCIM_Cubic, tangentMode=RCTM_User, time=0.9, value=1}]}}}
      CurveIndex=$System.NormalizedLoopAge
     )
    )
    UseOriginOffset=true
    Vortex Force Amount=&Lerp_Float(
     A=0
     B=500
     Alpha=&FloatFromCurve(
      FloatCurve=DataInterface() {propertyValues {Curve {keys [<at 0.4 0 RCIM_Cubic>,<at 0.7 1 RCIM_Cubic>]}}}
      CurveIndex=$System.NormalizedLoopAge
     )
    )
    Vortex Origin:NiagaraPosition=$System.SampleSkeletalMesh.SampledPosition
    Vortex Origin Offset:Vector3f=<0 0 25>
   )
   &PointAttractionForce(
    Attraction Radius=600
    Attraction Strength=&Multiply_Float(
     A=&FloatFromCurve(
      FloatCurve=DataInterface() {propertyValues {Curve {keys [<at 0.5 0 RCIM_Cubic>,<at 0.8 1 RCIM_Cubic>]}}}
      CurveIndex=$System.NormalizedLoopAge
     )
     B=&Lerp_Float(
      A=0
      B=500
      Alpha=&FloatFromCurve(FloatCurve=DataInterface() {propertyValues {Curve {keys [{tangentMode=RCTM_User, time=0.5, value=0}, {tangentMode=RCTM_User, time=1, value=1}]}}})
     )
    )
    Attractor Position:NiagaraPosition=$System.SampleSkeletalMesh.SampledPosition
    Attractor Position Offset:Vector3f=<0 0 35>
    Kill Within Radius=false
    Module.Use Falloff=true
    UseAttractorOffset=true
   )
   &ScaleColor(
    Scale RGB=&MakeVectorFromLinearColorRGB(Color=&ColorFromCurve(ColorCurve=DataInterface() {propertyValues {
        AlphaCurve {keys [<at 0 1 RCIM_Cubic>,<at 1 0 RCIM_Cubic>]}
        BlueCurve {keys [
          <at 0 0>
          {arriveTangent=0.0408, interpMode=RCIM_Cubic, leaveTangent=0.0408, time=0.2562, value=0.0249}
          <at 0.5961 0.03>
          <at 1 0.0408 RCIM_Cubic>
         ]}
        GreenCurve {keys [
          <at 0 0.9037>
          {arriveTangent=-0.8499, interpMode=RCIM_Cubic, leaveTangent=-0.8499, time=0.2562, value=0.3182}
          <at 0.5961 0.2113>
          <at 1 0.0538 RCIM_Cubic>
         ]}
        RedCurve {keys [
          <at 0 2>
          {arriveTangent=-1.5625, interpMode=RCIM_Cubic, leaveTangent=-1.5625, time=0.2562, value=1}
          <at 0.5961 0.5573>
          <at 1 0.4375 RCIM_Cubic>
         ]}
       }}))
    ScaleA=true
    ScaleRGB=true
   )
   &ScaleSpriteSize(Uniform Curve Sprite Scale=DataInterface() {propertyValues {Curve {keys [
        {interpMode=RCIM_Cubic, tangentMode=RCTM_User, time=0, value=0}
        {interpMode=RCIM_Cubic, tangentMode=RCTM_User, time=0.5, value=1}
        {tangentMode=RCTM_User, time=1, value=0}
       ]}}})
   &Drag(Drag=&RandomRangeFloat(Minimum=1, Maximum=2), Use Linear Drag=true, Use Rotational Drag=true)
   &CurlNoiseForce(Noise Strength=250, Pan Noise=false)
   &InheritSourceMovement(
    Applied Position Scale=&Multiply_VectorByFloat(
     Vector=<1 1 1>
     Float=&FloatFromCurve(FloatCurve=DataInterface() {propertyValues {Curve {keys [{interpMode=RCIM_Cubic, tangentMode=RCTM_User, time=0, value=1}, {interpMode=RCIM_Cubic, tangentMode=RCTM_User, time=1, value=0}]}}})
    )
    ApplyForce=false
    ApplyPosition=true
    ApplyPositionFalloff=false
   )
   &SolveForcesAndVelocity(Clamp Velocity=false, Limit Acceleration=false)
   &ScaleSpriteSizeBySpeed(Max Scale Factor=<1 4>, Min Scale Factor=<1 0>, Velocity Threshold=150)
  ]
  rendererClasses [&NiagaraSpriteRendererProperties]
  renderers [{rendererClass=&NiagaraSpriteRendererProperties}]
 }
}
summary {
 systemName=NS_Emote_Gatehound_Skulltrail
 emitters [
  {emitterName=SkullTrail, rendererClasses [&NiagaraRibbonRendererProperties]}
  {emitterName=Skulls_Main, rendererClasses [&NiagaraMeshRendererProperties, &NiagaraSpriteRendererProperties]}
  {emitterName=Sparks, simTarget=GPUComputeSim, rendererClasses [&NiagaraSpriteRendererProperties]}
  {emitterName=SmokeDetail, rendererClasses [&NiagaraSpriteRendererProperties]}
 ]
 userVariables [
  User.TrailWidth1 {type=NiagaraFloat, defaultValue=0, description=""}
  User.TrailWidth2 {type=NiagaraFloat, defaultValue=0, description=""}
  User.TrailWidth3 {type=NiagaraFloat, defaultValue=0, description=""}
  User.VariantSwitch {type=NiagaraFloat, defaultValue=0, description=""}
 ]
}
